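# Physics asset for the M_MED_Despair male body, covering its coat, tassels and zippers as dynamic bone chains alongside body capsules.
# Capsules cover root, pelvis, spine, thighs and calves, with a huge root sphere sitting low like a floor.
# MSTR anchor bodies attach to body bones, mostly with free rotation.
# Six four-link coat panels, tassels and some zippers hang from anchors with angle limits; grenade, boot, wrist and chest-zipper anchors have no child bodies.
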PhysicsAsset /BRCosmetics/Characters/Player/Male/Medium/Bodies/M_MED_Despair/Meshes/M_MED_Despair_FE_Physics
bodies {
 calf_l [{shapeType=Capsule, length=32.5258, radius=8.4213, center=<-23.4523 -1.1162 0.1776>, rotation=<rot 0.1246 84.5345 -88.5757>}]
 calf_r [{shapeType=Capsule, length=27.4442, radius=10.2443, center=<22.5428 0.7419 -0.7716>, rotation=<rot -0.3385 89.5329 -88.4634>}]
 dyn_backTassel_01 [{shapeType=Capsule, length=0, radius=3.0714, shapeName=dyn_backTassel_01_capsule, center=<1.8605 0 1.5>, rotation=<rot 90 0 0>}]
 dyn_backTassel_02 [{shapeType=Capsule, length=0, radius=3.0714, shapeName=dyn_backTassel_02_capsule, center=<1.8605 0 1.5>, rotation=<rot 90 0 0>}]
 dyn_backTassel_MSTR [{shapeType=Capsule, length=0.505, radius=0.505, shapeName=dyn_backTassel_MSTR_capsule, rotation=<rot 0 -90 0>}]
 dyn_buttZipper [{shapeType=Capsule, length=0, radius=3.6938, shapeName=dyn_buttZipper_capsule, center=<1.9451 0 1>, rotation=<rot 90 0 0>}]
 dyn_buttZipper_MSTR [{shapeType=Capsule, length=0.505, radius=0.505, shapeName=dyn_buttZipper_MSTR_capsule, rotation=<rot 0 -90 0>}]
 dyn_chestZipper_MSTR [{shapeType=Capsule, length=0.505, radius=0.505, shapeName=dyn_chestZipper_MSTR_capsule, rotation=<rot 0 -90 0>}]
 dyn_coat_bk_01 [{shapeType=Capsule, length=0, radius=10.1842, shapeName=dyn_coat_bk_01_capsule, center=<6.8336 0 8>, rotation=<rot 90 0 0>}]
 dyn_coat_bk_02 [{shapeType=Capsule, length=0.9084, radius=9.0916, shapeName=dyn_coat_bk_02_capsule, center=<6.8336 0 6>, rotation=<rot 90 0 0>}]
 dyn_coat_bk_03 [{shapeType=Capsule, length=1.9339, radius=8.0661, shapeName=dyn_coat_bk_03_capsule, center=<6.8336 0 3>, rotation=<rot 90 0 0>}]
 dyn_coat_bk_04 [{shapeType=Capsule, length=1.9339, radius=8.0661, shapeName=dyn_coat_bk_04_capsule, center=<6.8336 0 3>, rotation=<rot 90 0 0>}]
 dyn_coat_bk_MSTR [{shapeType=Capsule, length=0.505, radius=0.505, shapeName=dyn_coat_bk_MSTR_capsule, rotation=<rot 0 -90 0>}]
 dyn_coat_bk_l_01 [{shapeType=Capsule, length=0, radius=10.1842, shapeName=dyn_coat_bk_l_01_capsule, center=<6.8336 0 8>, rotation=<rot 90 0 0>}]
 dyn_coat_bk_l_02 [{shapeType=Capsule, length=0.9084, radius=9.0916, shapeName=dyn_coat_bk_l_02_capsule, center=<6.8336 0 6>, rotation=<rot 90 0 0>}]
 dyn_coat_bk_l_03 [{shapeType=Capsule, length=42.0243, radius=9.4147, shapeName=dyn_coat_bk_l_03_capsule, center=<6.8336 0 3>, rotation=<rot 0 0 90>}]
 dyn_coat_bk_l_04 [{shapeType=Capsule, length=42.0243, radius=9.4147, shapeName=dyn_coat_bk_l_04_capsule, center=<6.8336 0 3>, rotation=<rot 0 0 90>}]
 dyn_coat_bk_l_MSTR [{shapeType=Capsule, length=0.5, radius=0.5, shapeName=dyn_coat_bk_l_MSTR_capsule, rotation=<rot 0 -90 0>}]
 dyn_coat_fr_l_01 [{shapeType=Capsule, length=8.0328, radius=4.8805, shapeName=dyn_coat_fr_l_01_capsule, center=<6.8336 0 5>, rotation=<rot 90 0 0>}]
 dyn_coat_fr_l_02 [{shapeType=Capsule, length=8.0328, radius=4.8805, shapeName=dyn_coat_fr_l_02_capsule, center=<6.8336 0 4>, rotation=<rot 90 0 0>}]
 dyn_coat_fr_l_03 [{shapeType=Capsule, length=47.1927, radius=6.2291, shapeName=dyn_coat_fr_l_03_capsule, center=<6.8336 0 3>, rotation=<rot 0 0 90>}]
 dyn_coat_fr_l_04 [{shapeType=Capsule, length=47.1927, radius=6.2291, shapeName=dyn_coat_fr_l_04_capsule, center=<6.8336 0 3>, rotation=<rot 0 0 90>}]
 dyn_coat_fr_l_MSTR [{shapeType=Capsule, length=0.505, radius=0.505, shapeName=dyn_coat_fr_l_MSTR_capsule, rotation=<rot 0 -90 0>}]
 dyn_coat_fr_r_01 [{shapeType=Capsule, length=4.0452, radius=5.9548, shapeName=dyn_coat_fr_r_01_capsule, center=<6.8336 0 5>, rotation=<rot 90 0 0>}]
 dyn_coat_fr_r_02 [{shapeType=Capsule, length=5.4081, radius=4.5919, shapeName=dyn_coat_fr_r_02_capsule, center=<6.8336 0 4>, rotation=<rot 90 0 0>}]
 dyn_coat_fr_r_03 [{shapeType=Capsule, length=5.4081, radius=4.5919, shapeName=dyn_coat_fr_r_03_capsule, center=<6.8336 0 3>, rotation=<rot 90 0 0>}]
 dyn_coat_fr_r_04 [{shapeType=Capsule, length=5.4081, radius=4.5919, shapeName=dyn_coat_fr_r_04_capsule, center=<6.8336 0 3>, rotation=<rot 90 0 0>}]
 dyn_coat_fr_r_MSTR [{shapeType=Capsule, length=0.505, radius=0.505, shapeName=dyn_coat_fr_r_MSTR_capsule, rotation=<rot 0 -90 0>}]
 dyn_coat_l_01 [{shapeType=Capsule, length=16.3375, radius=5.8632, shapeName=dyn_coat_l_01_capsule, center=<6.8336 0 5>, rotation=<rot 0 0 90>}]
 dyn_coat_l_02 [{shapeType=Capsule, length=16.7643, radius=7.3277, shapeName=dyn_coat_l_02_capsule, center=<6.8336 0 5>, rotation=<rot 0 0 90>}]
 dyn_coat_l_03 [{shapeType=Capsule, length=30.0498, radius=9.4147, shapeName=dyn_coat_l_03_capsule, center=<6.8336 0 3>, rotation=<rot 0 0 90>}]
 dyn_coat_l_04 [{shapeType=Capsule, length=30.0498, radius=9.4147, shapeName=dyn_coat_l_04_capsule, center=<6.8336 0 3>, rotation=<rot 0 0 90>}]
 dyn_coat_l_MSTR [{shapeType=Capsule, length=0.505, radius=0.505, shapeName=dyn_coat_l_MSTR_capsule, rotation=<rot 0 -90 0>}]
 dyn_coat_r_01 [{shapeType=Capsule, length=0, radius=10.1842, shapeName=dyn_coat_r_01_capsule, center=<6.8336 0 8>, rotation=<rot 90 0 0>}]
 dyn_coat_r_02 [{shapeType=Capsule, length=0.9084, radius=9.0916, shapeName=dyn_coat_r_02_capsule, center=<6.8336 0 6>, rotation=<rot 90 0 0>}]
 dyn_coat_r_03 [{shapeType=Capsule, length=1.9339, radius=8.0661, shapeName=dyn_coat_r_03_capsule, center=<6.8336 0 3>, rotation=<rot 90 0 0>}]
 dyn_coat_r_04 [{shapeType=Capsule, length=1.9339, radius=8.0661, shapeName=dyn_coat_r_04_capsule, center=<6.8336 0 3>, rotation=<rot 90 0 0>}]
 dyn_coat_r_MSTR [{shapeType=Capsule, length=0.505, radius=0.505, shapeName=dyn_coat_r_MSTR_capsule, rotation=<rot 0 -90 0>}]
 dyn_gernade_01_MSTR [{shapeType=Capsule, length=0.505, radius=0.505, shapeName=dyn_gernade_01_MSTR_capsule, rotation=<rot 0 -90 0>}]
 dyn_gernade_02_MSTR [{shapeType=Capsule, length=0.505, radius=0.505, shapeName=dyn_gernade_02_MSTR_capsule, rotation=<rot 0 -90 0>}]
 dyn_gernade_03_MSTR [{shapeType=Capsule, length=0.505, radius=0.505, shapeName=dyn_gernade_03_MSTR_capsule, rotation=<rot 0 -90 0>}]
 dyn_gernade_04_MSTR [{shapeType=Capsule, length=0.505, radius=0.505, shapeName=dyn_gernade_04_MSTR_capsule, rotation=<rot 0 -90 0>}]
 dyn_l_bootLip_MSTR [{shapeType=Capsule, length=0.505, radius=0.505, shapeName=dyn_l_bootLip_MSTR_capsule, rotation=<rot 0 -90 0>}]
 dyn_l_bootTassel_MSTR [{shapeType=Capsule, length=0.505, radius=0.505, shapeName=dyn_l_bootTassel_MSTR_capsule, rotation=<rot 0 -90 0>}]
 dyn_l_thighZipper [{shapeType=Capsule, length=1.5149, radius=1.8747, shapeName=dyn_l_thighZipper_capsule, center=<1.9451 0 1>, rotation=<rot 90 0 0>}]
 dyn_l_thighZipper_MSTR [{shapeType=Capsule, length=0.505, radius=0.505, shapeName=dyn_l_thighZipper_MSTR_capsule, rotation=<rot 0 -90 0>}]
 dyn_l_wristStrap_MSTR [{shapeType=Capsule, length=0.505, radius=0.505, shapeName=dyn_l_wristStrap_MSTR_capsule, rotation=<rot 0 -90 0>}]
 dyn_r_bootLip_MSTR [{shapeType=Capsule, length=0.505, radius=0.505, shapeName=dyn_r_bootLip_MSTR_capsule, rotation=<rot 0 -90 0>}]
 dyn_r_bootTassel_MSTR [{shapeType=Capsule, length=0.505, radius=0.505, shapeName=dyn_r_bootTassel_MSTR_capsule, rotation=<rot 0 -90 0>}]
 dyn_r_thighZipper [{shapeType=Capsule, length=1.7756, radius=4.2244, shapeName=dyn_r_thighZipper_capsule, center=<3.5747 0 0>, rotation=<rot 90 0 0>}]
 dyn_r_thighZipper_MSTR [{shapeType=Capsule, length=0.505, radius=0.505, shapeName=dyn_r_thighZipper_MSTR_capsule, rotation=<rot 0 -90 0>}]
 dyn_torsoTassel_01 [{shapeType=Capsule, length=3.1979, radius=1.2133, shapeName=dyn_torsoTassel_01_capsule, center=<2.8556 0 0>, rotation=<rot 90 0 0>}]
 dyn_torsoTassel_02 [{shapeType=Capsule, length=3.1979, radius=1.2133, shapeName=dyn_torsoTassel_02_capsule, center=<2.0263 0 0>, rotation=<rot 90 0 0>}]
 dyn_torsoTassel_MSTR [{shapeType=Capsule, length=0.505, radius=0.505, shapeName=dyn_torsoTassel_MSTR_capsule, rotation=<rot 0 -90 0>}]
 pelvis [
  {shapeType=Capsule, length=6.8109, radius=10.4392, center=<-0.9227 -1.2812 0>}
  {shapeType=Capsule, length=8.5187, radius=10.849, shapeName=pelvis_capsule, center=<-3.9978 -1.0388 0>}
 ]
 root [
  {shapeType=Capsule, length=0, radius=187, shapeName=root_capsule, center=<-3.4923 -2.714 -174.8455>}
  {shapeType=Capsule, length=0, radius=27, shapeName=root_capsule, center=<10 0 0>, rotation=<rot 0 -90 0>}
 ]
 spine_01 [{shapeType=Capsule, length=10, radius=10, shapeName=spine_01_capsule}]
 spine_02 [{shapeType=Capsule, length=10, radius=10, shapeName=spine_02_capsule}]
 spine_03 [{shapeType=Capsule, length=10, radius=10, shapeName=spine_03_capsule}]
 spine_04 [{shapeType=Capsule, length=21.606, radius=13.4531, shapeName=spine_04_capsule}]
 spine_05 [{shapeType=Capsule, length=20.535, radius=14.5242, shapeName=spine_05_capsule, center=<-6.3081 0.7772 0>}]
 thigh_l [
  {shapeType=Capsule, length=27.8775, radius=8.1225, center=<-21.6102 0.8362 1.3426>, rotation=<rot 85 179.8044 179.9999>}
  {
   shapeType=Capsule
   length=19.1812
   radius=8.5559
   shapeName=thigh_l_capsule
   center=<-22.5963 3.4456 0.3807>
   rotation=<rot 85 179.8044 179.9999>
  }
  {
   shapeType=Capsule
   length=27.6482
   radius=8.3518
   shapeName=thigh_l_capsule
   center=<-21.3701 6.9221 1.5788>
   rotation=<rot 85 179.8044 179.9999>
  }
 ]
 thigh_r [
  {shapeType=Capsule, length=28.5812, radius=7.4188, center=<21.0533 -1.2399 -1.1849>, rotation=<rot 85 -180 -180>}
  {shapeType=Capsule, length=19.5428, radius=6.1648, shapeName=thigh_r_capsule, center=<21.1706 -1.2399 0.7329>, rotation=<rot 85 -180 -180>}
  {shapeType=Capsule, length=28.7244, radius=7.2756, shapeName=thigh_r_capsule, center=<21.2995 -6.6346 -2.0382>, rotation=<rot 85 -180 -180>}
 ]
}
constraints [
 {
  bone1Name=thigh_l
  bone2Name=pelvis
  swing1LimitDegrees=45
  swing1Motion=Free
  swing2LimitDegrees=45
  swing2Motion=Free
  twistLimitDegrees=45
  twistMotion=Free
 }
 {
  bone1Name=calf_l
  bone2Name=thigh_l
  swing1LimitDegrees=45
  swing1Motion=Free
  swing2LimitDegrees=45
  swing2Motion=Free
  twistLimitDegrees=45
  twistMotion=Free
 }
 {
  bone1Name=thigh_r
  bone2Name=pelvis
  swing1LimitDegrees=45
  swing1Motion=Free
  swing2LimitDegrees=45
  swing2Motion=Free
  twistLimitDegrees=45
  twistMotion=Free
 }
 {
  bone1Name=calf_r
  bone2Name=thigh_r
  swing1LimitDegrees=45
  swing1Motion=Free
  swing2LimitDegrees=45
  swing2Motion=Free
  twistLimitDegrees=45
  twistMotion=Free
 }
 {
  bone1Name=dyn_buttZipper_MSTR
  bone2Name=pelvis
  swing1LimitDegrees=45
  swing1Motion=Free
  swing2LimitDegrees=45
  swing2Motion=Free
  twistLimitDegrees=45
  twistMotion=Free
 }
 {
  bone1Name=dyn_coat_bk_MSTR
  bone2Name=pelvis
  swing1LimitDegrees=45
  swing1Motion=Free
  swing2LimitDegrees=45
  swing2Motion=Free
  twistLimitDegrees=45
  twistMotion=Free
 }
 {
  bone1Name=dyn_coat_bk_l_MSTR
  bone2Name=pelvis
  swing1LimitDegrees=45
  swing1Motion=Free
  swing2LimitDegrees=45
  swing2Motion=Free
  twistLimitDegrees=45
  twistMotion=Free
 }
 {
  bone1Name=dyn_coat_fr_l_MSTR
  bone2Name=pelvis
  swing1LimitDegrees=45
  swing1Motion=Free
  swing2LimitDegrees=45
  swing2Motion=Free
  twistLimitDegrees=45
  twistMotion=Free
 }
 {
  bone1Name=dyn_coat_fr_r_MSTR
  bone2Name=pelvis
  swing1LimitDegrees=45
  swing1Motion=Free
  swing2LimitDegrees=45
  swing2Motion=Free
  twistLimitDegrees=45
  twistMotion=Free
 }
 {
  bone1Name=dyn_coat_l_MSTR
  bone2Name=pelvis
  swing1LimitDegrees=45
  swing1Motion=Free
  swing2LimitDegrees=45
  swing2Motion=Free
  twistLimitDegrees=45
  twistMotion=Free
 }
 {
  bone1Name=dyn_coat_r_MSTR
  bone2Name=pelvis
  swing1LimitDegrees=45
  swing1Motion=Free
  swing2LimitDegrees=45
  swing2Motion=Free
  twistLimitDegrees=45
  twistMotion=Free
 }
 {
  bone1Name=dyn_gernade_01_MSTR
  bone2Name=pelvis
  swing1LimitDegrees=45
  swing1Motion=Free
  swing2LimitDegrees=45
  swing2Motion=Free
  twistLimitDegrees=45
  twistMotion=Free
 }
 {
  bone1Name=dyn_gernade_02_MSTR
  bone2Name=pelvis
  swing1LimitDegrees=45
  swing1Motion=Free
  swing2LimitDegrees=45
  swing2Motion=Free
  twistLimitDegrees=45
  twistMotion=Free
 }
 {
  bone1Name=dyn_gernade_03_MSTR
  bone2Name=pelvis
  swing1LimitDegrees=45
  swing1Motion=Free
  swing2LimitDegrees=45
  swing2Motion=Free
  twistLimitDegrees=45
  twistMotion=Free
 }
 {
  bone1Name=dyn_gernade_04_MSTR
  bone2Name=pelvis
  swing1LimitDegrees=45
  swing1Motion=Free
  swing2LimitDegrees=45
  swing2Motion=Free
  twistLimitDegrees=45
  twistMotion=Free
 }
 {
  bone1Name=dyn_torsoTassel_MSTR
  bone2Name=pelvis
  swing1LimitDegrees=45
  swing1Motion=Free
  swing2LimitDegrees=45
  swing2Motion=Free
  twistLimitDegrees=45
  twistMotion=Free
 }
 {
  bone1Name=dyn_l_wristStrap_MSTR
  bone2Name=spine_05
  swing1LimitDegrees=45
  swing1Motion=Free
  swing2LimitDegrees=45
  swing2Motion=Free
  twistLimitDegrees=45
  twistMotion=Free
 }
 {
  bone1Name=dyn_backTassel_MSTR
  bone2Name=spine_05
  swing1LimitDegrees=45
  swing1Motion=Free
  swing2LimitDegrees=45
  swing2Motion=Free
  twistLimitDegrees=45
  twistMotion=Free
 }
 {bone1Name=dyn_chestZipper_MSTR, bone2Name=spine_05, swing1LimitDegrees=45, swing2LimitDegrees=45, twistLimitDegrees=45}
 {
  bone1Name=dyn_l_bootLip_MSTR
  bone2Name=calf_l
  swing1LimitDegrees=45
  swing1Motion=Free
  swing2LimitDegrees=45
  swing2Motion=Free
  twistLimitDegrees=45
  twistMotion=Free
 }
 {
  bone1Name=dyn_l_bootTassel_MSTR
  bone2Name=calf_l
  swing1LimitDegrees=45
  swing1Motion=Free
  swing2LimitDegrees=45
  swing2Motion=Free
  twistLimitDegrees=45
  twistMotion=Free
 }
 {
  bone1Name=dyn_l_thighZipper_MSTR
  bone2Name=thigh_l
  swing1LimitDegrees=45
  swing1Motion=Free
  swing2LimitDegrees=45
  swing2Motion=Free
  twistLimitDegrees=45
  twistMotion=Free
 }
 {
  bone1Name=dyn_r_bootLip_MSTR
  bone2Name=calf_r
  swing1LimitDegrees=45
  swing1Motion=Free
  swing2LimitDegrees=45
  swing2Motion=Free
  twistLimitDegrees=45
  twistMotion=Free
 }
 {
  bone1Name=dyn_r_bootTassel_MSTR
  bone2Name=calf_r
  swing1LimitDegrees=45
  swing1Motion=Free
  swing2LimitDegrees=45
  swing2Motion=Free
  twistLimitDegrees=45
  twistMotion=Free
 }
 {
  bone1Name=dyn_r_thighZipper_MSTR
  bone2Name=thigh_r
  swing1LimitDegrees=45
  swing1Motion=Free
  swing2LimitDegrees=45
  swing2Motion=Free
  twistLimitDegrees=45
  twistMotion=Free
 }
 {
  bone1Name=pelvis
  bone2Name=root
  swing1LimitDegrees=45
  swing1Motion=Free
  swing2LimitDegrees=45
  swing2Motion=Free
  twistLimitDegrees=45
  twistMotion=Free
 }
 {bone1Name=dyn_coat_bk_01, bone2Name=dyn_coat_bk_MSTR, swing1LimitDegrees=20, swing2LimitDegrees=45, twistLimitDegrees=20}
 {bone1Name=dyn_coat_bk_02, bone2Name=dyn_coat_bk_01, swing1LimitDegrees=20, swing2LimitDegrees=45, twistLimitDegrees=20}
 {bone1Name=dyn_coat_bk_03, bone2Name=dyn_coat_bk_02, swing1LimitDegrees=20, swing2LimitDegrees=45, twistLimitDegrees=20}
 {bone1Name=dyn_coat_bk_04, bone2Name=dyn_coat_bk_03, swing1LimitDegrees=20, swing2LimitDegrees=45, twistLimitDegrees=20}
 {bone1Name=dyn_coat_bk_l_01, bone2Name=dyn_coat_bk_l_MSTR, swing1LimitDegrees=20, swing2LimitDegrees=45, twistLimitDegrees=20}
 {bone1Name=dyn_coat_bk_l_02, bone2Name=dyn_coat_bk_l_01, swing1LimitDegrees=20, swing2LimitDegrees=45, twistLimitDegrees=20}
 {bone1Name=dyn_coat_bk_l_03, bone2Name=dyn_coat_bk_l_02, swing1LimitDegrees=20, swing2LimitDegrees=45, twistLimitDegrees=20}
 {bone1Name=dyn_coat_bk_l_04, bone2Name=dyn_coat_bk_l_03, swing1LimitDegrees=20, swing2LimitDegrees=45, twistLimitDegrees=20}
 {bone1Name=dyn_coat_fr_l_02, bone2Name=dyn_coat_fr_l_01, swing1LimitDegrees=10, swing2LimitDegrees=10, twistLimitDegrees=10}
 {bone1Name=dyn_coat_fr_l_01, bone2Name=dyn_coat_fr_l_MSTR, swing1LimitDegrees=5, swing2LimitDegrees=5, twistLimitDegrees=5}
 {bone1Name=dyn_coat_fr_l_03, bone2Name=dyn_coat_fr_l_02, swing1LimitDegrees=30, swing2LimitDegrees=30, twistLimitDegrees=30}
 {bone1Name=dyn_coat_fr_l_04, bone2Name=dyn_coat_fr_l_03, swing1LimitDegrees=30, swing2LimitDegrees=30, twistLimitDegrees=30}
 {bone1Name=dyn_coat_fr_r_01, bone2Name=dyn_coat_fr_r_MSTR, swing1LimitDegrees=5, swing2LimitDegrees=5, twistLimitDegrees=5}
 {bone1Name=dyn_coat_fr_r_02, bone2Name=dyn_coat_fr_r_01, swing1LimitDegrees=5, swing2LimitDegrees=5, twistLimitDegrees=5}
 {bone1Name=dyn_coat_fr_r_03, bone2Name=dyn_coat_fr_r_02, swing1LimitDegrees=5, swing2LimitDegrees=5, twistLimitDegrees=5}
 {bone1Name=dyn_coat_fr_r_04, bone2Name=dyn_coat_fr_r_03, swing1LimitDegrees=5, swing2LimitDegrees=5, twistLimitDegrees=5}
 {bone1Name=dyn_coat_l_01, bone2Name=dyn_coat_l_MSTR, swing1LimitDegrees=20, swing2LimitDegrees=45, twistLimitDegrees=20}
 {bone1Name=dyn_coat_l_02, bone2Name=dyn_coat_l_01, swing1LimitDegrees=20, swing2LimitDegrees=45, twistLimitDegrees=20}
 {bone1Name=dyn_coat_l_03, bone2Name=dyn_coat_l_02, swing1LimitDegrees=20, swing2LimitDegrees=45, twistLimitDegrees=20}
 {bone1Name=dyn_coat_l_04, bone2Name=dyn_coat_l_03, swing1LimitDegrees=20, swing2LimitDegrees=45, twistLimitDegrees=20}
 {bone1Name=dyn_coat_r_01, bone2Name=dyn_coat_r_MSTR, swing1LimitDegrees=20, swing2LimitDegrees=45, twistLimitDegrees=20}
 {bone1Name=dyn_coat_r_02, bone2Name=dyn_coat_r_01, swing1LimitDegrees=20, swing2LimitDegrees=45, twistLimitDegrees=20}
 {bone1Name=dyn_coat_r_03, bone2Name=dyn_coat_r_02, swing1LimitDegrees=20, swing2LimitDegrees=45, twistLimitDegrees=20}
 {bone1Name=dyn_coat_r_04, bone2Name=dyn_coat_r_03, swing1LimitDegrees=20, swing2LimitDegrees=45, twistLimitDegrees=20}
 {bone1Name=dyn_buttZipper, bone2Name=dyn_buttZipper_MSTR, swing1LimitDegrees=45, swing2LimitDegrees=45, twistLimitDegrees=5}
 {bone1Name=dyn_l_thighZipper, bone2Name=dyn_l_thighZipper_MSTR, swing1LimitDegrees=45, swing2LimitDegrees=45, twistLimitDegrees=5}
 {bone1Name=dyn_r_thighZipper, bone2Name=dyn_r_thighZipper_MSTR, swing1LimitDegrees=45, swing2LimitDegrees=45, twistLimitDegrees=5}
 {
  bone1Name=spine_05
  bone2Name=spine_04
  swing1LimitDegrees=45
  swing1Motion=Free
  swing2LimitDegrees=45
  swing2Motion=Free
  twistLimitDegrees=45
  twistMotion=Free
 }
 {
  bone1Name=spine_04
  bone2Name=spine_03
  swing1LimitDegrees=45
  swing1Motion=Free
  swing2LimitDegrees=45
  swing2Motion=Free
  twistLimitDegrees=45
  twistMotion=Free
 }
 {bone1Name=dyn_backTassel_01, bone2Name=dyn_backTassel_MSTR, swing1LimitDegrees=40, swing2LimitDegrees=40, twistLimitDegrees=10}
 {bone1Name=dyn_backTassel_02, bone2Name=dyn_backTassel_01, swing1LimitDegrees=40, swing2LimitDegrees=40, twistLimitDegrees=10}
 {bone1Name=dyn_torsoTassel_02, bone2Name=dyn_torsoTassel_01, swing1LimitDegrees=45, swing2LimitDegrees=45, twistLimitDegrees=45}
 {bone1Name=dyn_torsoTassel_01, bone2Name=dyn_torsoTassel_MSTR, swing1LimitDegrees=45, swing2LimitDegrees=45, twistLimitDegrees=45}
 {
  bone1Name=spine_03
  bone2Name=spine_02
  swing1LimitDegrees=45
  swing1Motion=Free
  swing2LimitDegrees=45
  swing2Motion=Free
  twistLimitDegrees=45
  twistMotion=Free
 }
 {
  bone1Name=spine_02
  bone2Name=spine_01
  swing1LimitDegrees=45
  swing1Motion=Free
  swing2LimitDegrees=45
  swing2Motion=Free
  twistLimitDegrees=45
  twistMotion=Free
 }
 {
  bone1Name=spine_01
  bone2Name=pelvis
  swing1LimitDegrees=45
  swing1Motion=Free
  swing2LimitDegrees=45
  swing2Motion=Free
  twistLimitDegrees=45
  twistMotion=Free
 }
]
