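# Physics asset for the M_MED_Despair male player body cosmetic, covering its long coat and small dangling accessories.
# Capsules on pelvis, thighs and calves, plus two zero-length root capsules (one very large), form the body volumes.
# Six four-segment coat chains hang from pelvis-linked MSTR anchors with angle-limited joints.
# Tiny MSTR bodies for zippers, tassels, grenades, straps and boot lips attach to pelvis or legs with free rotation.
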PhysicsAsset /BRCosmetics/Characters/Player/Male/Medium/Bodies/M_MED_Despair/Meshes/M_MED_Despair_Physics
bodies {
 calf_l [{shapeType=Capsule, length=38.4137, radius=8.5449, center=<-23.4523 -1.1162 0.1776>, rotation=<rot 0.1246 84.5345 -88.5757>}]
 calf_r [{shapeType=Capsule, length=18.9596, radius=15.9417, center=<22.5428 0.7419 -0.7716>, rotation=<rot -0.3385 89.5329 -88.4634>}]
 dyn_backTassel_MSTR [{shapeType=Capsule, length=0.505, radius=0.505, shapeName=dyn_backTassel_MSTR_capsule, rotation=<rot 0 -90 0>}]
 dyn_buttZipper_MSTR [{shapeType=Capsule, length=0.505, radius=0.505, shapeName=dyn_buttZipper_MSTR_capsule, rotation=<rot 0 -90 0>}]
 dyn_chestZipper_MSTR [{shapeType=Capsule, length=0.505, radius=0.505, shapeName=dyn_chestZipper_MSTR_capsule, rotation=<rot 0 -90 0>}]
 dyn_coat_bk_01 [{shapeType=Capsule, length=13.0113, radius=14.9887, shapeName=dyn_coat_bk_01_capsule, center=<7.5 0 15>, rotation=<rot 0 0 90>}]
 dyn_coat_bk_02 [{shapeType=Capsule, length=13.0113, radius=14.9887, shapeName=dyn_coat_bk_02_capsule, center=<7.5 0 15>, rotation=<rot 0 0 90>}]
 dyn_coat_bk_03 [{shapeType=Capsule, length=15, radius=13, shapeName=dyn_coat_bk_03_capsule, center=<7.5 0 9.5>, rotation=<rot 0 0 90>}]
 dyn_coat_bk_04 [{shapeType=Capsule, length=15, radius=13, shapeName=dyn_coat_bk_04_capsule, center=<7.5 0 9.5>, rotation=<rot 0 0 90>}]
 dyn_coat_bk_MSTR [{shapeType=Capsule, length=0.505, radius=0.505, shapeName=dyn_coat_bk_MSTR_capsule, rotation=<rot 0 -90 0>}]
 dyn_coat_bk_l_01 [{shapeType=Capsule, length=13.5967, radius=14.4033, shapeName=dyn_coat_bk_l_01_capsule, center=<7.5 0 15>, rotation=<rot 0 0 90>}]
 dyn_coat_bk_l_02 [{shapeType=Capsule, length=13.5967, radius=14.4033, shapeName=dyn_coat_bk_l_02_capsule, center=<7.5 0 15>, rotation=<rot 0 0 90>}]
 dyn_coat_bk_l_03 [{shapeType=Capsule, length=15, radius=13, shapeName=dyn_coat_bk_l_03_capsule, center=<7.5 0 9.5>, rotation=<rot 0 0 90>}]
 dyn_coat_bk_l_04 [{shapeType=Capsule, length=15, radius=13, shapeName=dyn_coat_bk_l_04_capsule, center=<7.5 0 9.5>, rotation=<rot 0 0 90>}]
 dyn_coat_bk_l_MSTR [{shapeType=Capsule, length=0.5, radius=0.5, shapeName=dyn_coat_bk_l_MSTR_capsule, rotation=<rot 0 -90 0>}]
 dyn_coat_fr_l_01 [{shapeType=Capsule, length=12.1809, radius=6.1851, shapeName=dyn_coat_fr_l_01_capsule, center=<7.5 0 9.5>, rotation=<rot 0 0 90>}]
 dyn_coat_fr_l_02 [{shapeType=Capsule, length=13.0935, radius=10.9065, shapeName=dyn_coat_fr_l_02_capsule, center=<7.5 0 9.5>, rotation=<rot 0 0 90>}]
 dyn_coat_fr_l_03 [{shapeType=Capsule, length=16.4308, radius=13.3192, shapeName=dyn_coat_fr_l_03_capsule, center=<7.5 0 9.5>, rotation=<rot 0 0 90>}]
 dyn_coat_fr_l_04 [{shapeType=Capsule, length=14.6571, radius=15.0929, shapeName=dyn_coat_fr_l_04_capsule, center=<7.5 0 9.5>, rotation=<rot 0 0 90>}]
 dyn_coat_fr_l_MSTR [{shapeType=Capsule, length=0.505, radius=0.505, shapeName=dyn_coat_fr_l_MSTR_capsule, rotation=<rot 0 -90 0>}]
 dyn_coat_fr_r_01 [{shapeType=Capsule, length=8.3032, radius=8.6968, shapeName=dyn_coat_fr_r_01_capsule, center=<7.5 0 9.5>, rotation=<rot 0 0 90>}]
 dyn_coat_fr_r_02 [{shapeType=Capsule, length=15.9782, radius=8.0218, shapeName=dyn_coat_fr_r_02_capsule, center=<7.5 0 9.5>, rotation=<rot 0 0 90>}]
 dyn_coat_fr_r_03 [{shapeType=Capsule, length=17.8259, radius=11.9241, shapeName=dyn_coat_fr_r_03_capsule, center=<7.5 0 9.5>, rotation=<rot 0 0 90>}]
 dyn_coat_fr_r_04 [{shapeType=Capsule, length=15.5506, radius=14.1994, shapeName=dyn_coat_fr_r_04_capsule, center=<7.5 0 9.5>, rotation=<rot 0 0 90>}]
 dyn_coat_fr_r_MSTR [{shapeType=Capsule, length=0.505, radius=0.505, shapeName=dyn_coat_fr_r_MSTR_capsule, rotation=<rot 0 -90 0>}]
 dyn_coat_l_01 [{shapeType=Capsule, length=8.773, radius=8.227, shapeName=dyn_coat_l_01_capsule, center=<7.5 0 9.5>, rotation=<rot 0 0 107.2565>}]
 dyn_coat_l_02 [{shapeType=Capsule, length=13.1933, radius=10.8067, shapeName=dyn_coat_l_02_capsule, center=<7.5 0 9.5>, rotation=<rot 0 0 90>}]
 dyn_coat_l_03 [{shapeType=Capsule, length=15.532, radius=14.218, shapeName=dyn_coat_l_03_capsule, center=<7.5 0 9.5>, rotation=<rot 0 0 90>}]
 dyn_coat_l_04 [{shapeType=Capsule, length=15.522, radius=14.228, shapeName=dyn_coat_l_04_capsule, center=<7.5 0 9.5>, rotation=<rot 0 0 90>}]
 dyn_coat_l_MSTR [{shapeType=Capsule, length=0.505, radius=0.505, shapeName=dyn_coat_l_MSTR_capsule, rotation=<rot 0 -90 0>}]
 dyn_coat_r_01 [{shapeType=Capsule, length=14.8627, radius=8.464, shapeName=dyn_coat_r_01_capsule, center=<7.2181 0 9>, rotation=<rot 0 0 90>}]
 dyn_coat_r_02 [{shapeType=Capsule, length=13.1194, radius=10.2074, shapeName=dyn_coat_r_02_capsule, center=<7.2181 0 9>, rotation=<rot 0 0 90>}]
 dyn_coat_r_03 [{shapeType=Capsule, length=8.8875, radius=14.4393, shapeName=dyn_coat_r_03_capsule, center=<7.2181 0 9>, rotation=<rot 0 0 90>}]
 dyn_coat_r_04 [{shapeType=Capsule, length=19.3323, radius=16.2327, shapeName=dyn_coat_r_04_capsule, center=<7.2181 0 9>, rotation=<rot 0 0 138.5018>}]
 dyn_coat_r_MSTR [{shapeType=Capsule, length=0.505, radius=0.505, shapeName=dyn_coat_r_MSTR_capsule, rotation=<rot 0 -90 0>}]
 dyn_gernade_01_MSTR [{shapeType=Capsule, length=0.505, radius=0.505, shapeName=dyn_gernade_01_MSTR_capsule, rotation=<rot 0 -90 0>}]
 dyn_gernade_02_MSTR [{shapeType=Capsule, length=0.505, radius=0.505, shapeName=dyn_gernade_02_MSTR_capsule, rotation=<rot 0 -90 0>}]
 dyn_gernade_03_MSTR [{shapeType=Capsule, length=0.505, radius=0.505, shapeName=dyn_gernade_03_MSTR_capsule, rotation=<rot 0 -90 0>}]
 dyn_gernade_04_MSTR [{shapeType=Capsule, length=0.505, radius=0.505, shapeName=dyn_gernade_04_MSTR_capsule, rotation=<rot 0 -90 0>}]
 dyn_l_bootLip_MSTR [{shapeType=Capsule, length=0.505, radius=0.505, shapeName=dyn_l_bootLip_MSTR_capsule, rotation=<rot 0 -90 0>}]
 dyn_l_bootTassel_MSTR [{shapeType=Capsule, length=0.505, radius=0.505, shapeName=dyn_l_bootTassel_MSTR_capsule, rotation=<rot 0 -90 0>}]
 dyn_l_thighZipper_MSTR [{shapeType=Capsule, length=0.505, radius=0.505, shapeName=dyn_l_thighZipper_MSTR_capsule, rotation=<rot 0 -90 0>}]
 dyn_l_wristStrap_MSTR [{shapeType=Capsule, length=0.505, radius=0.505, shapeName=dyn_l_wristStrap_MSTR_capsule, rotation=<rot 0 -90 0>}]
 dyn_r_bootLip_MSTR [{shapeType=Capsule, length=0.505, radius=0.505, shapeName=dyn_r_bootLip_MSTR_capsule, rotation=<rot 0 -90 0>}]
 dyn_r_bootTassel_MSTR [{shapeType=Capsule, length=0.505, radius=0.505, shapeName=dyn_r_bootTassel_MSTR_capsule, rotation=<rot 0 -90 0>}]
 dyn_r_thighZipper_MSTR [{shapeType=Capsule, length=0.505, radius=0.505, shapeName=dyn_r_thighZipper_MSTR_capsule, rotation=<rot 0 -90 0>}]
 dyn_torsoTassel_MSTR [{shapeType=Capsule, length=0.505, radius=0.505, shapeName=dyn_torsoTassel_MSTR_capsule, rotation=<rot 0 -90 0>}]
 pelvis [
  {shapeType=Capsule, length=4.6315, radius=12.6186, center=<-0.9227 -1.2812 0>}
  {shapeType=Capsule, length=6.3393, radius=13.0284, shapeName=pelvis_capsule, center=<-3.9978 -3.5061 0>}
 ]
 root [
  {shapeType=Capsule, length=0, radius=187, shapeName=root_capsule, center=<-3.4923 -2.714 -174.8455>}
  {shapeType=Capsule, length=0, radius=27, shapeName=root_capsule, center=<10 0 0>, rotation=<rot 0 -90 0>}
 ]
 thigh_l [
  {shapeType=Capsule, length=26.0811, radius=9.9189, center=<-21.6102 0.8362 1.3426>, rotation=<rot 85 179.8044 179.9999>}
  {
   shapeType=Capsule
   length=19.1812
   radius=8.5559
   shapeName=thigh_l_capsule
   center=<-22.5963 3.4456 0.3807>
   rotation=<rot 85 179.8044 179.9999>
  }
  {
   shapeType=Capsule
   length=25.8519
   radius=10.1482
   shapeName=thigh_l_capsule
   center=<-21.3701 6.9221 1.5788>
   rotation=<rot 85 179.8044 179.9999>
  }
 ]
 thigh_r [
  {shapeType=Capsule, length=24.788, radius=11.212, center=<21.0533 -1.2399 -1.1849>, rotation=<rot 85 -180 -180>}
  {shapeType=Capsule, length=17.5603, radius=8.1473, shapeName=thigh_r_capsule, center=<21.1706 -1.2399 0.7329>, rotation=<rot 85 -180 -180>}
  {shapeType=Capsule, length=24.9311, radius=11.0688, shapeName=thigh_r_capsule, center=<21.2995 -6.6346 -2.0382>, rotation=<rot 85 -180 -180>}
 ]
}
constraints [
 {
  bone1Name=thigh_l
  bone2Name=pelvis
  swing1LimitDegrees=45
  swing1Motion=Free
  swing2LimitDegrees=45
  swing2Motion=Free
  twistLimitDegrees=45
  twistMotion=Free
 }
 {
  bone1Name=calf_l
  bone2Name=thigh_l
  swing1LimitDegrees=45
  swing1Motion=Free
  swing2LimitDegrees=45
  swing2Motion=Free
  twistLimitDegrees=45
  twistMotion=Free
 }
 {
  bone1Name=thigh_r
  bone2Name=pelvis
  swing1LimitDegrees=45
  swing1Motion=Free
  swing2LimitDegrees=45
  swing2Motion=Free
  twistLimitDegrees=45
  twistMotion=Free
 }
 {
  bone1Name=calf_r
  bone2Name=thigh_r
  swing1LimitDegrees=45
  swing1Motion=Free
  swing2LimitDegrees=45
  swing2Motion=Free
  twistLimitDegrees=45
  twistMotion=Free
 }
 {
  bone1Name=dyn_buttZipper_MSTR
  bone2Name=pelvis
  swing1LimitDegrees=45
  swing1Motion=Free
  swing2LimitDegrees=45
  swing2Motion=Free
  twistLimitDegrees=45
  twistMotion=Free
 }
 {
  bone1Name=dyn_coat_bk_MSTR
  bone2Name=pelvis
  swing1LimitDegrees=45
  swing1Motion=Free
  swing2LimitDegrees=45
  swing2Motion=Free
  twistLimitDegrees=45
  twistMotion=Free
 }
 {
  bone1Name=dyn_coat_bk_l_MSTR
  bone2Name=pelvis
  swing1LimitDegrees=45
  swing1Motion=Free
  swing2LimitDegrees=45
  swing2Motion=Free
  twistLimitDegrees=45
  twistMotion=Free
 }
 {
  bone1Name=dyn_coat_fr_l_MSTR
  bone2Name=pelvis
  swing1LimitDegrees=45
  swing1Motion=Free
  swing2LimitDegrees=45
  swing2Motion=Free
  twistLimitDegrees=45
  twistMotion=Free
 }
 {
  bone1Name=dyn_coat_fr_r_MSTR
  bone2Name=pelvis
  swing1LimitDegrees=45
  swing1Motion=Free
  swing2LimitDegrees=45
  swing2Motion=Free
  twistLimitDegrees=45
  twistMotion=Free
 }
 {
  bone1Name=dyn_coat_l_MSTR
  bone2Name=pelvis
  swing1LimitDegrees=45
  swing1Motion=Free
  swing2LimitDegrees=45
  swing2Motion=Free
  twistLimitDegrees=45
  twistMotion=Free
 }
 {
  bone1Name=dyn_coat_r_MSTR
  bone2Name=pelvis
  swing1LimitDegrees=45
  swing1Motion=Free
  swing2LimitDegrees=45
  swing2Motion=Free
  twistLimitDegrees=45
  twistMotion=Free
 }
 {
  bone1Name=dyn_gernade_01_MSTR
  bone2Name=pelvis
  swing1LimitDegrees=45
  swing1Motion=Free
  swing2LimitDegrees=45
  swing2Motion=Free
  twistLimitDegrees=45
  twistMotion=Free
 }
 {
  bone1Name=dyn_gernade_02_MSTR
  bone2Name=pelvis
  swing1LimitDegrees=45
  swing1Motion=Free
  swing2LimitDegrees=45
  swing2Motion=Free
  twistLimitDegrees=45
  twistMotion=Free
 }
 {
  bone1Name=dyn_gernade_03_MSTR
  bone2Name=pelvis
  swing1LimitDegrees=45
  swing1Motion=Free
  swing2LimitDegrees=45
  swing2Motion=Free
  twistLimitDegrees=45
  twistMotion=Free
 }
 {
  bone1Name=dyn_gernade_04_MSTR
  bone2Name=pelvis
  swing1LimitDegrees=45
  swing1Motion=Free
  swing2LimitDegrees=45
  swing2Motion=Free
  twistLimitDegrees=45
  twistMotion=Free
 }
 {
  bone1Name=dyn_torsoTassel_MSTR
  bone2Name=pelvis
  swing1LimitDegrees=45
  swing1Motion=Free
  swing2LimitDegrees=45
  swing2Motion=Free
  twistLimitDegrees=45
  twistMotion=Free
 }
 {
  bone1Name=dyn_l_wristStrap_MSTR
  bone2Name=pelvis
  swing1LimitDegrees=45
  swing1Motion=Free
  swing2LimitDegrees=45
  swing2Motion=Free
  twistLimitDegrees=45
  twistMotion=Free
 }
 {
  bone1Name=dyn_backTassel_MSTR
  bone2Name=pelvis
  swing1LimitDegrees=45
  swing1Motion=Free
  swing2LimitDegrees=45
  swing2Motion=Free
  twistLimitDegrees=45
  twistMotion=Free
 }
 {
  bone1Name=dyn_chestZipper_MSTR
  bone2Name=pelvis
  swing1LimitDegrees=45
  swing1Motion=Free
  swing2LimitDegrees=45
  swing2Motion=Free
  twistLimitDegrees=45
  twistMotion=Free
 }
 {
  bone1Name=dyn_l_bootLip_MSTR
  bone2Name=calf_l
  swing1LimitDegrees=45
  swing1Motion=Free
  swing2LimitDegrees=45
  swing2Motion=Free
  twistLimitDegrees=45
  twistMotion=Free
 }
 {
  bone1Name=dyn_l_bootTassel_MSTR
  bone2Name=calf_l
  swing1LimitDegrees=45
  swing1Motion=Free
  swing2LimitDegrees=45
  swing2Motion=Free
  twistLimitDegrees=45
  twistMotion=Free
 }
 {
  bone1Name=dyn_l_thighZipper_MSTR
  bone2Name=thigh_l
  swing1LimitDegrees=45
  swing1Motion=Free
  swing2LimitDegrees=45
  swing2Motion=Free
  twistLimitDegrees=45
  twistMotion=Free
 }
 {
  bone1Name=dyn_r_bootLip_MSTR
  bone2Name=calf_r
  swing1LimitDegrees=45
  swing1Motion=Free
  swing2LimitDegrees=45
  swing2Motion=Free
  twistLimitDegrees=45
  twistMotion=Free
 }
 {
  bone1Name=dyn_r_bootTassel_MSTR
  bone2Name=calf_r
  swing1LimitDegrees=45
  swing1Motion=Free
  swing2LimitDegrees=45
  swing2Motion=Free
  twistLimitDegrees=45
  twistMotion=Free
 }
 {
  bone1Name=dyn_r_thighZipper_MSTR
  bone2Name=thigh_r
  swing1LimitDegrees=45
  swing1Motion=Free
  swing2LimitDegrees=45
  swing2Motion=Free
  twistLimitDegrees=45
  twistMotion=Free
 }
 {bone1Name=dyn_coat_bk_01, bone2Name=dyn_coat_bk_MSTR, swing1LimitDegrees=15, swing2LimitDegrees=45, twistLimitDegrees=30}
 {bone1Name=dyn_coat_bk_02, bone2Name=dyn_coat_bk_01, swing1LimitDegrees=15, swing2LimitDegrees=45, twistLimitDegrees=30}
 {bone1Name=dyn_coat_bk_03, bone2Name=dyn_coat_bk_02, swing1LimitDegrees=15, swing2LimitDegrees=45, twistLimitDegrees=30}
 {bone1Name=dyn_coat_bk_04, bone2Name=dyn_coat_bk_03, swing1LimitDegrees=15, swing2LimitDegrees=45, twistLimitDegrees=30}
 {bone1Name=dyn_coat_bk_l_01, bone2Name=dyn_coat_bk_l_MSTR, swing1LimitDegrees=15, swing2LimitDegrees=45, twistLimitDegrees=30}
 {bone1Name=dyn_coat_bk_l_02, bone2Name=dyn_coat_bk_l_01, swing1LimitDegrees=15, swing2LimitDegrees=45, twistLimitDegrees=30}
 {bone1Name=dyn_coat_bk_l_03, bone2Name=dyn_coat_bk_l_02, swing1LimitDegrees=15, swing2LimitDegrees=45, twistLimitDegrees=30}
 {bone1Name=dyn_coat_bk_l_04, bone2Name=dyn_coat_bk_l_03, swing1LimitDegrees=15, swing2LimitDegrees=45, twistLimitDegrees=30}
 {bone1Name=dyn_coat_fr_l_01, bone2Name=dyn_coat_fr_l_MSTR, swing1LimitDegrees=20, swing2LimitDegrees=45, twistLimitDegrees=15}
 {bone1Name=dyn_coat_fr_l_02, bone2Name=dyn_coat_fr_l_01, swing1LimitDegrees=20, swing2LimitDegrees=45, twistLimitDegrees=15}
 {bone1Name=dyn_coat_fr_l_03, bone2Name=dyn_coat_fr_l_02, swing1LimitDegrees=20, swing2LimitDegrees=45, twistLimitDegrees=15}
 {bone1Name=dyn_coat_fr_l_04, bone2Name=dyn_coat_fr_l_03, swing1LimitDegrees=20, swing2LimitDegrees=45, twistLimitDegrees=15}
 {bone1Name=dyn_coat_fr_r_01, bone2Name=dyn_coat_fr_r_MSTR, swing1LimitDegrees=20, swing2LimitDegrees=45, twistLimitDegrees=15}
 {bone1Name=dyn_coat_fr_r_02, bone2Name=dyn_coat_fr_r_01, swing1LimitDegrees=20, swing2LimitDegrees=45, twistLimitDegrees=15}
 {bone1Name=dyn_coat_fr_r_03, bone2Name=dyn_coat_fr_r_02, swing1LimitDegrees=30, swing2LimitDegrees=45, twistLimitDegrees=30}
 {bone1Name=dyn_coat_fr_r_04, bone2Name=dyn_coat_fr_r_03, swing1LimitDegrees=30, swing2LimitDegrees=45, twistLimitDegrees=30}
 {bone1Name=dyn_coat_l_02, bone2Name=dyn_coat_l_01, swing1LimitDegrees=20, swing2LimitDegrees=45, twistLimitDegrees=15}
 {bone1Name=dyn_coat_l_01, bone2Name=dyn_coat_l_MSTR, swing1LimitDegrees=20, swing2LimitDegrees=45, twistLimitDegrees=15}
 {bone1Name=dyn_coat_l_03, bone2Name=dyn_coat_l_02, swing1LimitDegrees=20, swing2LimitDegrees=45, twistLimitDegrees=15}
 {bone1Name=dyn_coat_l_04, bone2Name=dyn_coat_l_03, swing1LimitDegrees=20, swing2LimitDegrees=45, twistLimitDegrees=15}
 {
  bone1Name=pelvis
  bone2Name=root
  swing1LimitDegrees=45
  swing1Motion=Free
  swing2LimitDegrees=45
  swing2Motion=Free
  twistLimitDegrees=45
  twistMotion=Free
 }
 {bone1Name=dyn_coat_r_01, bone2Name=dyn_coat_r_MSTR, swing1LimitDegrees=30, swing2LimitDegrees=45, twistLimitDegrees=30}
 {bone1Name=dyn_coat_r_02, bone2Name=dyn_coat_r_01, swing1LimitDegrees=30, swing2LimitDegrees=45, twistLimitDegrees=30}
 {bone1Name=dyn_coat_r_03, bone2Name=dyn_coat_r_02, swing1LimitDegrees=30, swing2LimitDegrees=45, twistLimitDegrees=30}
 {bone1Name=dyn_coat_r_04, bone2Name=dyn_coat_r_03, swing1LimitDegrees=30, swing2LimitDegrees=45, twistLimitDegrees=30}
]
