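# Animation Blueprint for the M_MED_Despair player body cosmetic, adding secondary motion to coat, boot, tassel, zipper and shoulder bones.
# It copies the attached parent's pose, then applies deform and control rigs plus zipper constraints.
# Coat RigidBody simulation switches between front-end, standard and crouching setups, with acceleration-driven bone offsets when crouched.
# Constraints then tie coat bones to the right thigh and back-left chain before per-bone AnimDynamics runs.
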AnimBlueprint /BRCosmetics/Characters/Player/Male/Medium/Bodies/M_MED_Despair/Meshes/M_MED_Despair_AnimBP
target_skeleton=&/M_MED_Despair_Skeleton
graphs [
 AnimGraph {nodes [
   n1 = RigidBody!(
    ComponentPose=Local To Component(LocalPose=Use cached pose 'PreSim'())
    Alpha
    OverrideWorldGravity:Vector=Get GravityOverride()
    ExternalForce
   ) -> (Pose) {
    bClampLinearTranslationLimitToRefPose=true
    bOverrideWorldGravity=true
    overridePhysicsAsset=&/M_MED_Despair_Physics
    simulationSpace=BaseBoneSpace
    alphaBoolBlend {blendOption=Linear}
    baseBoneRef {boneName=pelvis}
    componentAppliedLinearAccClamp=<3000 3000 3000>
    componentLinearAccScale=<0.3 0.3 0.5>
    componentLinearVelScale=<0.3 0.3 0.5>
    simSpaceSettings {
     maxAngularAcceleration=1000
     maxAngularVelocity=3
     maxLinearAcceleration=700
     maxLinearVelocity=500
     velocityScaleZ=0.1
     worldAlpha=1
     externalLinearDragV=<0.2 0.2>
    }
   }
   n2 = RigidBody(ComponentPose=Local To Component(LocalPose=Use cached pose 'PreSim'()), OverrideWorldGravity=Get GravityOverride()) {
    bClampLinearTranslationLimitToRefPose=true
    bOverrideWorldGravity=true
    overridePhysicsAsset=&/M_MED_Despair_FE_Physics
    simulationSpace=BaseBoneSpace
    alphaBoolBlend {blendOption=Linear}
    baseBoneRef {boneName=root}
    componentAppliedLinearAccClamp=<5000 5000 5000>
    componentLinearAccScale=<0.3 0.3 0.5>
    componentLinearVelScale=<0.3 0.3 0.5>
    simSpaceSettings {
     maxAngularAcceleration=1000
     maxAngularVelocity=3
     maxLinearAcceleration=700
     maxLinearVelocity=500
     velocityScaleZ=0.1
     worldAlpha=1
     externalLinearDragV=<0.2 0.2>
    }
   }
   n3 = "Transform (Modify) Bone - Bone: dyn_coat_l_MSTR"!(ComponentPose=@n2, bAlphaBoolEnabled:bool=True, Translation="0,0,2.5", Rotation="0,0,0", Scale="1,1,1") -> (Pose) {
    alphaInputType=Bool
    rotationMode=BMM_Additive
    rotationSpace=BCS_BoneSpace
    scaleMode=BMM_Additive
    scaleSpace=BCS_BoneSpace
    translationMode=BMM_Additive
    translationSpace=BCS_BoneSpace
    alphaBoolBlend {blendInTime=0.1, blendOption=Linear, blendOutTime=0.1}
    alphaScaleBiasClamp {bInterpResult=true}
    boneToModify {boneName=dyn_coat_l_MSTR}
    translation {z=2.5}
   }
   n4 = "Transform (Modify) Bone - Bone: dyn_coat_bk_l_MSTR"!(ComponentPose=@n3, bAlphaBoolEnabled:bool=True, Translation="0,0,2.5", Rotation="0,0,0", Scale="1,1,1") -> (Pose) {
    alphaInputType=Bool
    rotationMode=BMM_Additive
    rotationSpace=BCS_BoneSpace
    scaleMode=BMM_Additive
    scaleSpace=BCS_BoneSpace
    translationMode=BMM_Additive
    translationSpace=BCS_BoneSpace
    alphaBoolBlend {blendInTime=0.1, blendOption=Linear, blendOutTime=0.1}
    alphaScaleBiasClamp {bInterpResult=true}
    boneToModify {boneName=dyn_coat_bk_l_MSTR}
    translation {z=2.5}
   }
   n5 = "Transform (Modify) Bone - Bone: dyn_coat_bk_MSTR"!(ComponentPose=@n4, bAlphaBoolEnabled:bool=True, Translation="0,0,2.5", Rotation="0,0,0", Scale="1,1,1") -> (Pose) {
    alphaInputType=Bool
    rotationMode=BMM_Additive
    rotationSpace=BCS_BoneSpace
    scaleMode=BMM_Additive
    scaleSpace=BCS_BoneSpace
    translationMode=BMM_Additive
    translationSpace=BCS_BoneSpace
    alphaBoolBlend {blendInTime=0.1, blendOption=Linear, blendOutTime=0.1}
    alphaScaleBiasClamp {bInterpResult=true}
    boneToModify {boneName=dyn_coat_bk_MSTR}
    translation {z=2.5}
   }
   n6 = "Transform (Modify) Bone - Bone: dyn_coat_r_MSTR"!(ComponentPose=@n5, bAlphaBoolEnabled:bool=True, Translation="0,0,2.5", Rotation="0,0,0", Scale="1,1,1") -> (Pose) {
    alphaInputType=Bool
    rotationMode=BMM_Additive
    rotationSpace=BCS_BoneSpace
    scaleMode=BMM_Additive
    scaleSpace=BCS_BoneSpace
    translationMode=BMM_Additive
    translationSpace=BCS_BoneSpace
    alphaBoolBlend {blendInTime=0.1, blendOption=Linear, blendOutTime=0.1}
    alphaScaleBiasClamp {bInterpResult=true}
    boneToModify {boneName=dyn_coat_r_MSTR}
    translation {z=2.5}
   }
   n7 = Blend Poses by bool(
    BlendPose_0=Component To Local(ComponentPose=@n6)
    BlendPose_1=Component To Local(ComponentPose=@n1)
    BlendTime_1=1
    bActiveValue=Get bIsInFrontEnd()
   ) {blendType=Linear}
   n8 = Get bIsAccelerating2D()
   n9 = "Transform (Modify) Bone - Bone: dyn_coat_fr_r_MSTR"!(
    ComponentPose=Local To Component(LocalPose=Use cached pose 'PreSim'())
    bAlphaBoolEnabled:bool=@n8
    Translation="0,0,0"
    Rotation="50,0,0"
    Scale="1,1,1"
   ) -> (Pose) {
    alphaInputType=Bool
    rotationMode=BMM_Additive
    rotationSpace=BCS_BoneSpace
    translationMode=BMM_Additive
    translationSpace=BCS_BoneSpace
    alphaBoolBlend {blendInTime=0.1, blendOption=Linear, blendOutTime=0.1}
    alphaScaleBiasClamp {bInterpResult=true}
    boneToModify {boneName=dyn_coat_fr_r_MSTR}
    rotation {pitch=50}
   }
   n10 = "Transform (Modify) Bone - Bone: dyn_coat_bk_l_MSTR"(ComponentPose=@n9, bAlphaBoolEnabled=@n8, Translation="0,-3,0", Rotation="0,30,0.000", Scale="1,1,1") {
    alphaInputType=Bool
    rotationMode=BMM_Additive
    rotationSpace=BCS_BoneSpace
    translationMode=BMM_Additive
    translationSpace=BCS_BoneSpace
    alphaBoolBlend {blendInTime=0.1, blendOption=Linear, blendOutTime=0.1}
    alphaScaleBiasClamp {bInterpResult=true}
    boneToModify {boneName=dyn_coat_bk_l_MSTR}
    rotation {yaw=30}
    translation {y=-3}
   }
   n11 = "Transform (Modify) Bone - Bone: dyn_coat_bk_MSTR"(ComponentPose=@n10, bAlphaBoolEnabled=@n8, Translation="0,-5,0", Rotation="0,30,0.000", Scale="1,1,1") {
    alphaInputType=Bool
    rotationMode=BMM_Additive
    rotationSpace=BCS_BoneSpace
    translationMode=BMM_Additive
    translationSpace=BCS_BoneSpace
    alphaBoolBlend {blendInTime=0.1, blendOption=Linear, blendOutTime=0.1}
    alphaScaleBiasClamp {bInterpResult=true}
    boneToModify {boneName=dyn_coat_bk_MSTR}
    rotation {yaw=30}
    translation {y=-5}
   }
   n12 = "Transform (Modify) Bone - Bone: dyn_coat_r_MSTR"(ComponentPose=@n11, bAlphaBoolEnabled=@n8, Translation="0,0,5", Rotation="0,30,0.000", Scale="1,1,1") {
    alphaInputType=Bool
    rotationMode=BMM_Additive
    rotationSpace=BCS_BoneSpace
    translationMode=BMM_Additive
    translationSpace=BCS_BoneSpace
    alphaBoolBlend {blendInTime=0.1, blendOption=Linear, blendOutTime=0.1}
    alphaScaleBiasClamp {bInterpResult=true}
    boneToModify {boneName=dyn_coat_r_MSTR}
    rotation {yaw=30}
    translation {z=5}
   }
   n13 = RigidBody(ComponentPose=@n12, OverrideWorldGravity=Get GravityOverride()) {
    bClampLinearTranslationLimitToRefPose=true
    bOverrideWorldGravity=true
    overridePhysicsAsset=&M_MED_Despair_Physics
    simulationSpace=BaseBoneSpace
    alphaBoolBlend {blendOption=Linear}
    baseBoneRef {boneName=pelvis}
    componentAppliedLinearAccClamp=<4000 4000 4000>
    componentLinearAccScale=<0.4 0.4 0.4>
    componentLinearVelScale=<0.4 0.4 0.4>
    simSpaceSettings {
     maxAngularAcceleration=1000
     maxAngularVelocity=3
     maxLinearAcceleration=700
     maxLinearVelocity=500
     velocityScaleZ=0
     worldAlpha=0.5
     externalLinearDragV=<0.2 0.2>
    }
   }
   n14 = "Constraint - dyn_coat_fr_r_01 (R)"(ComponentPose=@n13, Alpha=0.5, ConstraintWeights_0=1) {
    alphaBoolBlend {blendOption=Linear}
    boneToModify {boneName=dyn_coat_fr_r_01}
    constraintSetup [{transformType=Rotation, targetBone {boneName=thigh_r}}]
    constraintWeights [1]
   }
   n15 = "Constraint - dyn_coat_fr_r_02 (R)"(ComponentPose=@n14, Alpha=0.5, ConstraintWeights_0=1) {
    alphaBoolBlend {blendOption=Linear}
    boneToModify {boneName=dyn_coat_fr_r_02}
    constraintSetup [{transformType=Rotation, targetBone {boneName=thigh_r}}]
    constraintWeights [1]
   }
   n16 = "Constraint - dyn_coat_fr_r_03 (R)"(ComponentPose=@n15, Alpha=0.5, ConstraintWeights_0=1) {
    alphaBoolBlend {blendOption=Linear}
    boneToModify {boneName=dyn_coat_fr_r_03}
    constraintSetup [{transformType=Rotation, targetBone {boneName=thigh_r}}]
    constraintWeights [1]
   }
   n17 = Component To Local(ComponentPose=@n16)
   n18 = Get bIsCrouching()
   n19 = "Constraint - dyn_coat_bk_MSTR (R)"!(
    ComponentPose=Local To Component(LocalPose=Blend Poses by bool(BlendPose_0=@n17, BlendPose_1=@n7, BlendTime_1=1, bActiveValue=@n18) {blendType=Linear})
    bAlphaBoolEnabled:bool=True
    ConstraintWeights_0=1
   ) -> (Pose) {
    alphaInputType=Bool
    alphaBoolBlend {blendInTime=0.1, blendOption=Linear, blendOutTime=0.1}
    boneToModify {boneName=dyn_coat_bk_MSTR}
    constraintSetup [{transformType=Rotation, targetBone {boneName=dyn_coat_bk_l_MSTR}}]
    constraintWeights [1]
   }
   n20 = "Constraint - dyn_coat_bk_01 (R)"!(ComponentPose=@n19, bAlphaBoolEnabled:bool=True, ConstraintWeights_0=1) -> (Pose) {
    alphaInputType=Bool
    alphaBoolBlend {blendInTime=0.1, blendOption=Linear, blendOutTime=0.1}
    boneToModify {boneName=dyn_coat_bk_01}
    constraintSetup [{transformType=Rotation, targetBone {boneName=dyn_coat_bk_l_01}}]
    constraintWeights [1]
   }
   n21 = "Constraint - dyn_coat_bk_02 (R)"!(ComponentPose=@n20, bAlphaBoolEnabled:bool=True, ConstraintWeights_0=1) -> (Pose) {
    alphaInputType=Bool
    alphaBoolBlend {blendInTime=0.1, blendOption=Linear, blendOutTime=0.1}
    boneToModify {boneName=dyn_coat_bk_02}
    constraintSetup [{transformType=Rotation, targetBone {boneName=dyn_coat_bk_l_02}}]
    constraintWeights [1]
   }
   n22 = "Constraint - dyn_coat_bk_03 (R)"!(ComponentPose=@n21, bAlphaBoolEnabled:bool=True, ConstraintWeights_0=1) -> (Pose) {
    alphaInputType=Bool
    alphaBoolBlend {blendInTime=0.1, blendOption=Linear, blendOutTime=0.1}
    boneToModify {boneName=dyn_coat_bk_03}
    constraintSetup [{transformType=Rotation, targetBone {boneName=dyn_coat_bk_l_03}}]
    constraintWeights [1]
   }
   n23 = "Constraint - dyn_coat_bk_04 (R)"!(ComponentPose=@n22, bAlphaBoolEnabled:bool=True, ConstraintWeights_0=1) -> (Pose) {
    alphaInputType=Bool
    alphaBoolBlend {blendInTime=0.1, blendOption=Linear, blendOutTime=0.1}
    boneToModify {boneName=dyn_coat_bk_04}
    constraintSetup [{transformType=Rotation, targetBone {boneName=dyn_coat_bk_l_04}}]
    constraintWeights [1]
   }
   n24 = "Constraint - dyn_coat_r_MSTR (R)"!(ComponentPose=@n23, bAlphaBoolEnabled:bool=True, ConstraintWeights_0=1) -> (Pose) {
    alphaInputType=Bool
    alphaBoolBlend {blendInTime=0.1, blendOption=Linear, blendOutTime=0.1}
    boneToModify {boneName=dyn_coat_r_MSTR}
    constraintSetup [{transformType=Rotation, targetBone {boneName=thigh_r}}]
    constraintWeights [1]
   }
   n25 = "Constraint - dyn_coat_r_01 (R)"!(ComponentPose=@n24, bAlphaBoolEnabled:bool=True, ConstraintWeights_0=1) -> (Pose) {
    alphaInputType=Bool
    alphaBoolBlend {blendInTime=0.1, blendOption=Linear, blendOutTime=0.1}
    boneToModify {boneName=dyn_coat_r_01}
    constraintSetup [{transformType=Rotation, targetBone {boneName=thigh_r}}]
    constraintWeights [1]
   }
   n26 = "Constraint - dyn_coat_r_02 (R)"!(ComponentPose=@n25, bAlphaBoolEnabled:bool=True, ConstraintWeights_0=1) -> (Pose) {
    alphaInputType=Bool
    alphaBoolBlend {blendInTime=0.1, blendOption=Linear, blendOutTime=0.1}
    boneToModify {boneName=dyn_coat_r_02}
    constraintSetup [{transformType=Rotation, targetBone {boneName=thigh_r}}]
    constraintWeights [1]
   }
   n27 = "Constraint - dyn_coat_r_03 (R)"!(ComponentPose=@n26, bAlphaBoolEnabled:bool=True, ConstraintWeights_0=1) -> (Pose) {
    alphaInputType=Bool
    alphaBoolBlend {blendInTime=0.1, blendOption=Linear, blendOutTime=0.1}
    boneToModify {boneName=dyn_coat_r_03}
    constraintSetup [{transformType=Rotation, targetBone {boneName=thigh_r}}]
    constraintWeights [1]
   }
   n28 = "Constraint - dyn_coat_r_04 (R)"!(ComponentPose=@n27, bAlphaBoolEnabled:bool=True, ConstraintWeights_0=1) -> (Pose) {
    alphaInputType=Bool
    alphaBoolBlend {blendInTime=0.1, blendOption=Linear, blendOutTime=0.1}
    boneToModify {boneName=dyn_coat_r_04}
    constraintSetup [{transformType=Rotation, targetBone {boneName=thigh_r}}]
    constraintWeights [1]
   }
   n29 = "Constraint - dyn_coat_fr_r_MSTR (R)"!(ComponentPose=@n28, bAlphaBoolEnabled:bool=True, ConstraintWeights_0=1) -> (Pose) {
    alphaInputType=Bool
    alphaBoolBlend {blendInTime=0.1, blendOption=Linear, blendOutTime=0.1}
    boneToModify {boneName=dyn_coat_fr_r_MSTR}
    constraintSetup [{transformType=Rotation, targetBone {boneName=thigh_r}}]
    constraintWeights [1]
   }
   n30 = "Constraint - dyn_coat_fr_r_01 (R)"!(ComponentPose=@n29, bAlphaBoolEnabled:bool=True, ConstraintWeights_0=1) -> (Pose) {
    alphaInputType=Bool
    alphaBoolBlend {blendInTime=0.1, blendOption=Linear, blendOutTime=0.1}
    boneToModify {boneName=dyn_coat_fr_r_01}
    constraintSetup [{transformType=Rotation, targetBone {boneName=thigh_r}}]
    constraintWeights [1]
   }
   n31 = "Constraint - dyn_coat_fr_r_02 (R)"!(ComponentPose=@n30, bAlphaBoolEnabled:bool=True, ConstraintWeights_0=1) -> (Pose) {
    alphaInputType=Bool
    alphaBoolBlend {blendInTime=0.1, blendOption=Linear, blendOutTime=0.1}
    boneToModify {boneName=dyn_coat_fr_r_02}
    constraintSetup [{transformType=Rotation, targetBone {boneName=thigh_r}}]
    constraintWeights [1]
   }
   n32 = "Constraint - dyn_coat_fr_r_03 (R)"!(ComponentPose=@n31, bAlphaBoolEnabled:bool=True, ConstraintWeights_0=1) -> (Pose) {
    alphaInputType=Bool
    alphaBoolBlend {blendInTime=0.1, blendOption=Linear, blendOutTime=0.1}
    boneToModify {boneName=dyn_coat_fr_r_03}
    constraintSetup [{transformType=Rotation, targetBone {boneName=thigh_r}}]
    constraintWeights [1]
   }
   n33 = "Constraint - dyn_coat_fr_r_03 (R)"(ComponentPose=@n32, bAlphaBoolEnabled=True, ConstraintWeights_0=1) {
    alphaInputType=Bool
    alphaBoolBlend {blendInTime=0.1, blendOption=Linear, blendOutTime=0.1}
    boneToModify {boneName=dyn_coat_fr_r_03}
    constraintSetup [{transformType=Rotation, targetBone {boneName=thigh_r}}]
    constraintWeights [1]
   }
   n34 = "Constraint - dyn_coat_fr_r_04 (R)"!(ComponentPose=@n33, bAlphaBoolEnabled:bool=True, ConstraintWeights_0=1) -> (Pose) {
    alphaInputType=Bool
    alphaBoolBlend {blendInTime=0.1, blendOption=Linear, blendOutTime=0.1}
    boneToModify {boneName=dyn_coat_fr_r_04}
    constraintSetup [{transformType=Rotation, targetBone {boneName=thigh_r}}]
    constraintWeights [1]
   }
   n35 = "Transform (Modify) Bone - Bone: dyn_coat_fr_r_MSTR"(ComponentPose=@n34, bAlphaBoolEnabled=True, Translation="0,0,7", Rotation="0,0,0", Scale="1,1,1") {
    alphaInputType=Bool
    rotationMode=BMM_Additive
    rotationSpace=BCS_BoneSpace
    translationMode=BMM_Additive
    translationSpace=BCS_BoneSpace
    alphaBoolBlend {blendInTime=0.1, blendOption=Linear, blendOutTime=0.1}
    alphaScaleBiasClamp {bInterpResult=true}
    boneToModify {boneName=dyn_coat_fr_r_MSTR}
    translation {z=7}
   }
   n36 = "Transform (Modify) Bone - Bone: dyn_coat_r_MSTR"(ComponentPose=@n35, bAlphaBoolEnabled=True, Translation="0,0,7", Rotation="0,0,0", Scale="1,1,1") {
    alphaInputType=Bool
    rotationMode=BMM_Additive
    rotationSpace=BCS_BoneSpace
    translationMode=BMM_Additive
    translationSpace=BCS_BoneSpace
    alphaBoolBlend {blendInTime=0.1, blendOption=Linear, blendOutTime=0.1}
    alphaScaleBiasClamp {bInterpResult=true}
    boneToModify {boneName=dyn_coat_r_MSTR}
    translation {z=7}
   }
   n37 = Anim Dynamics - Bone: dyn_l_bootLip(ComponentPose=@n36, Alpha=1, ExternalForce="0,0,0") {
    angularDampingOverride=0.999
    angularSpringConstant=1100
    bAngularSpring=true
    bGravityOverrideInSimSpace=true
    bOverrideAngularDamping=true
    bUseGravityOverride=true
    gravityScale=0.35
    simulationSpace=RootRelative
    alphaBoolBlend {blendOption=Linear}
    boundBone {boneName=dyn_l_bootLip}
    componentAppliedLinearAccClamp=<5000 5000 5000>
    componentLinearAccScale=<1 1 1>
    componentLinearVelScale=<1 1 1>
    gravityOverride {z=-980}
    physicsBodyDefinitions [{
      sphereCollisionRadius=0
      boundBone {boneName=dyn_l_bootLip}
      boxExtents=<20 10 10>
      constraintSetup {angularLimitsMax=<0 -20 15>, angularLimitsMin=<0 -10 -15>, angularTarget=<1 0 0>}
      localJointOffset=<1 0 0>
     }]
   }
   n38 = Anim Dynamics - Bone: dyn_r_bootLip(ComponentPose=@n37, Alpha=1, ExternalForce="0,0,0") {
    angularDampingOverride=0.999
    angularSpringConstant=1100
    bAngularSpring=true
    bGravityOverrideInSimSpace=true
    bOverrideAngularDamping=true
    bUseGravityOverride=true
    gravityScale=0.35
    simulationSpace=RootRelative
    alphaBoolBlend {blendOption=Linear}
    boundBone {boneName=dyn_r_bootLip}
    componentAppliedLinearAccClamp=<5000 5000 5000>
    componentLinearAccScale=<1 1 1>
    componentLinearVelScale=<1 1 1>
    gravityOverride {z=-980}
    physicsBodyDefinitions [{
      sphereCollisionRadius=0
      boundBone {boneName=dyn_r_bootLip}
      boxExtents=<20 10 10>
      constraintSetup {angularLimitsMax=<0 -30 15>, angularLimitsMin=<0 -20 -15>, angularTarget=<1 0 0>}
      localJointOffset=<1 0 0>
     }]
   }
   n39 = Anim Dynamics - Bone: dyn_l_bootTassel(ComponentPose=@n38, Alpha=1, ExternalForce="0,0,0") {
    angularDampingOverride=1
    angularSpringConstant=12000
    bAngularSpring=true
    bGravityOverrideInSimSpace=true
    bOverrideAngularDamping=true
    bUseGravityOverride=true
    gravityScale=0.35
    simulationSpace=RootRelative
    alphaBoolBlend {blendOption=Linear}
    boundBone {boneName=dyn_l_bootTassel}
    componentAppliedLinearAccClamp=<6000 6000 6000>
    componentLinearAccScale=<1 1 1>
    componentLinearVelScale=<1 1 1>
    gravityOverride {z=-980}
    physicsBodyDefinitions [{
      sphereCollisionRadius=0
      boundBone {boneName=dyn_l_bootTassel}
      boxExtents=<20 10 10>
      constraintSetup {angularLimitsMax=<0 20 0>, angularLimitsMin=<0 -20 0>, angularTarget=<1 0 0>}
      localJointOffset=<4 0 0>
     }]
   }
   n40 = Anim Dynamics - Bone: dyn_r_bootTassel(ComponentPose=@n39, Alpha=1, ExternalForce="0,0,0") {
    angularDampingOverride=1
    angularSpringConstant=12000
    bAngularSpring=true
    bGravityOverrideInSimSpace=true
    bOverrideAngularDamping=true
    bUseGravityOverride=true
    gravityScale=0.35
    simulationSpace=RootRelative
    alphaBoolBlend {blendOption=Linear}
    boundBone {boneName=dyn_r_bootTassel}
    componentAppliedLinearAccClamp=<6000 6000 6000>
    componentLinearAccScale=<1 1 1>
    componentLinearVelScale=<1 1 1>
    gravityOverride {z=-980}
    physicsBodyDefinitions [{
      sphereCollisionRadius=0
      boundBone {boneName=dyn_r_bootTassel}
      boxExtents=<20 10 10>
      constraintSetup {angularLimitsMax=<0 20 0>, angularLimitsMin=<0 -20 0>, angularTarget=<1 0 0>}
      localJointOffset=<4 0 0>
     }]
   }
   n41 = Anim Dynamics - Bone: dyn_backTassel_01(ComponentPose=@n40, Alpha=0.9, ExternalForce="0,0,0") {
    angularDampingOverride=1
    bGravityOverrideInSimSpace=true
    bOverrideAngularDamping=true
    simulationSpace=RootRelative
    alphaBoolBlend {blendOption=Linear}
    boundBone {boneName=dyn_backTassel_01}
    componentAppliedLinearAccClamp=<5000 5000 5000>
    componentLinearAccScale=<1.1 1.1 1.1>
    componentLinearVelScale=<1.1 1.1 1.1>
    physicsBodyDefinitions [{
      sphereCollisionRadius=0
      boundBone {boneName=dyn_backTassel_01}
      boxExtents=<8 15 15>
      constraintSetup {angularLimitsMax=<2 25 45>, angularLimitsMin=<-2 -80 -45>, angularTarget=<0 0 0>}
      localJointOffset=<1.5 -0.05 0.1>
     }]
   }
   n42 = Anim Dynamics - Bone: dyn_torsoTassel_01(ComponentPose=@n41, Alpha=1, ExternalForce="0,0,0") {
    angularDampingOverride=1
    bGravityOverrideInSimSpace=true
    bOverrideAngularDamping=true
    simulationSpace=RootRelative
    alphaBoolBlend {blendOption=Linear}
    boundBone {boneName=dyn_torsoTassel_01}
    componentAppliedLinearAccClamp=<5000 5000 5000>
    componentLinearAccScale=<1.25 1.25 1.25>
    componentLinearVelScale=<1.25 1.25 1.25>
    physicsBodyDefinitions [{
      sphereCollisionRadius=0
      boundBone {boneName=dyn_torsoTassel_01}
      boxExtents=<8 15 15>
      constraintSetup {angularLimitsMax=<2 15 45>, angularLimitsMin=<-2 0 -45>, angularTarget=<0 0 0>}
      localJointOffset=<1.5 -0.05 0.1>
     }]
   }
   n43 = Anim Dynamics - Bone: dyn_l_wristStrap(ComponentPose=@n42, Alpha=1, ExternalForce="0,0,0") {
    angularDampingOverride=1
    bGravityOverrideInSimSpace=true
    bOverrideAngularDamping=true
    simulationSpace=RootRelative
    alphaBoolBlend {blendOption=Linear}
    boundBone {boneName=dyn_l_wristStrap}
    componentAppliedLinearAccClamp=<5000 5000 5000>
    componentLinearAccScale=<1.25 1.25 1.25>
    componentLinearVelScale=<1.25 1.25 1.25>
    physicsBodyDefinitions [{
      sphereCollisionRadius=0
      boundBone {boneName=dyn_l_wristStrap}
      boxExtents=<8 15 15>
      constraintSetup {angularLimitsMax=<50 50 50>, angularLimitsMin=<-50 -50 -50>, angularTarget=<0 0 0>}
      localJointOffset=<1.5 -0.05 0.1>
     }]
   }
   n44 = Anim Dynamics - Bone: dyn_chestZipper(ComponentPose=@n43, Alpha=1, ExternalForce="0,0,0") {
    angularDampingOverride=0.9999
    bGravityOverrideInSimSpace=true
    bOverrideAngularDamping=true
    simulationSpace=RootRelative
    alphaBoolBlend {blendOption=Linear}
    boundBone {boneName=dyn_chestZipper}
    componentAppliedLinearAccClamp=<5000 5000 5000>
    componentLinearAccScale=<1.25 1.25 1.25>
    componentLinearVelScale=<1.25 1.25 1.25>
    physicsBodyDefinitions [{
      sphereCollisionRadius=0
      boundBone {boneName=dyn_chestZipper}
      boxExtents=<3 10 10>
      constraintSetup {angularLimitsMax=<2 -5 45>, angularLimitsMin=<-2 -50 -45>, angularTarget=<0 0 0>}
      localJointOffset=<1.5 -0.05 0.1>
     }]
   }
   n45 = Anim Dynamics - Bone: dyn_l_thighZipper(ComponentPose=@n44, Alpha=1, ExternalForce="0,0,0") {
    angularDampingOverride=0.9999
    bGravityOverrideInSimSpace=true
    bOverrideAngularDamping=true
    simulationSpace=RootRelative
    alphaBoolBlend {blendOption=Linear}
    boundBone {boneName=dyn_l_thighZipper}
    componentAppliedLinearAccClamp=<5000 5000 5000>
    componentLinearAccScale=<1.25 1.25 1.25>
    componentLinearVelScale=<1.25 1.25 1.25>
    physicsBodyDefinitions [{
      sphereCollisionRadius=0
      boundBone {boneName=dyn_l_thighZipper}
      boxExtents=<3 10 10>
      constraintSetup {angularLimitsMax=<2 0 45>, angularLimitsMin=<-2 -10 -45>, angularTarget=<0 0 0>}
      localJointOffset=<1.5 -0.05 0.1>
     }]
   }
   n46 = Anim Dynamics - Bone: dyn_r_thighZipper(ComponentPose=@n45, Alpha=1, ExternalForce="0,0,0") {
    angularDampingOverride=0.9999
    bGravityOverrideInSimSpace=true
    bOverrideAngularDamping=true
    simulationSpace=RootRelative
    alphaBoolBlend {blendOption=Linear}
    boundBone {boneName=dyn_r_thighZipper}
    componentAppliedLinearAccClamp=<5000 5000 5000>
    componentLinearAccScale=<1.25 1.25 1.25>
    componentLinearVelScale=<1.25 1.25 1.25>
    physicsBodyDefinitions [{
      sphereCollisionRadius=0
      boundBone {boneName=dyn_r_thighZipper}
      boxExtents=<3 10 10>
      constraintSetup {angularLimitsMax=<2 0 45>, angularLimitsMin=<-2 -10 -45>, angularTarget=<0 0 0>}
      localJointOffset=<1.5 -0.05 0.1>
     }]
   }
   n47 = Anim Dynamics - Bone: dyn_buttZipper(ComponentPose=@n46, Alpha=1, ExternalForce="0,0,0") {
    angularDampingOverride=0.9999
    bGravityOverrideInSimSpace=true
    bOverrideAngularDamping=true
    simulationSpace=RootRelative
    alphaBoolBlend {blendOption=Linear}
    boundBone {boneName=dyn_buttZipper}
    componentAppliedLinearAccClamp=<5000 5000 5000>
    componentLinearAccScale=<1.25 1.25 1.25>
    componentLinearVelScale=<1.25 1.25 1.25>
    physicsBodyDefinitions [{
      sphereCollisionRadius=0
      boundBone {boneName=dyn_buttZipper}
      boxExtents=<3 10 10>
      constraintSetup {angularLimitsMax=<2 0 45>, angularLimitsMin=<-2 -50 -45>, angularTarget=<0 0 0>}
      localJointOffset=<1.5 -0.05 0.1>
     }]
   }
   n48 = Anim Dynamics - Bone: dyn_shldr_j_l(ComponentPose=@n47, Alpha=0.1, ExternalForce="0,0,0") {
    angularDampingOverride=0.999
    angularSpringConstant=1500
    bAngularSpring=true
    bGravityOverrideInSimSpace=true
    bOverrideAngularDamping=true
    bUseGravityOverride=true
    gravityScale=0.35
    simulationSpace=RootRelative
    alphaBoolBlend {blendOption=Linear}
    boundBone {boneName=dyn_shldr_j_l}
    componentLinearAccScale=<1 1 1>
    componentLinearVelScale=<1 1 1>
    gravityOverride {z=-980}
    physicsBodyDefinitions [{
      sphereCollisionRadius=0
      boundBone {boneName=dyn_shldr_j_l}
      boxExtents=<25 15 15>
      constraintSetup {angularLimitsMax=<40 40 40>, angularLimitsMin=<-40 -40 -40>, angularTarget=<1 0 0>}
      localJointOffset=<8 0 0>
     }]
   }
   n49 = Anim Dynamics - Bone: dyn_shldr_e_l(ComponentPose=@n48, Alpha=0.15, ExternalForce="0,0,0") {
    angularDampingOverride=0.999
    angularSpringConstant=3000
    bAngularSpring=true
    bGravityOverrideInSimSpace=true
    bOverrideAngularDamping=true
    bUseGravityOverride=true
    gravityScale=0.35
    simulationSpace=RootRelative
    alphaBoolBlend {blendOption=Linear}
    boundBone {boneName=dyn_shldr_e_l}
    componentLinearAccScale=<1 1 1>
    componentLinearVelScale=<1 1 1>
    gravityOverride {z=-980}
    physicsBodyDefinitions [{
      sphereCollisionRadius=0
      boundBone {boneName=dyn_shldr_e_l}
      boxExtents=<25 15 15>
      constraintSetup {angularLimitsMax=<40 40 40>, angularLimitsMin=<-40 -40 -40>, angularTarget=<1 0 0>}
      localJointOffset=<8 0 0>
     }]
   }
   n50 = Anim Dynamics - Bone: dyn_shldr_i_l(ComponentPose=@n49, Alpha=0.2, ExternalForce="0,0,0") {
    angularDampingOverride=0.999
    angularSpringConstant=3000
    bAngularSpring=true
    bGravityOverrideInSimSpace=true
    bOverrideAngularDamping=true
    bUseGravityOverride=true
    gravityScale=0.35
    simulationSpace=RootRelative
    alphaBoolBlend {blendOption=Linear}
    boundBone {boneName=dyn_shldr_i_l}
    componentLinearAccScale=<1 1 1>
    componentLinearVelScale=<1 1 1>
    gravityOverride {z=-980}
    physicsBodyDefinitions [{
      sphereCollisionRadius=0
      boundBone {boneName=dyn_shldr_i_l}
      boxExtents=<25 15 15>
      constraintSetup {angularLimitsMax=<40 40 40>, angularLimitsMin=<-40 -40 -40>, angularTarget=<1 0 0>}
      localJointOffset=<8 0 0>
     }]
   }
   n51 = Anim Dynamics - Bone: dyn_shldr_b_l(ComponentPose=@n50, Alpha=0.2, ExternalForce="0,0,0") {
    angularDampingOverride=0.999
    angularSpringConstant=1500
    bAngularSpring=true
    bGravityOverrideInSimSpace=true
    bOverrideAngularDamping=true
    bUseGravityOverride=true
    gravityScale=0.35
    simulationSpace=RootRelative
    alphaBoolBlend {blendOption=Linear}
    boundBone {boneName=dyn_shldr_b_l}
    componentLinearAccScale=<1 1 1>
    componentLinearVelScale=<1 1 1>
    gravityOverride {z=-980}
    physicsBodyDefinitions [{
      sphereCollisionRadius=0
      boundBone {boneName=dyn_shldr_b_l}
      boxExtents=<25 15 15>
      constraintSetup {angularLimitsMax=<40 40 40>, angularLimitsMin=<-40 -40 -40>, angularTarget=<1 0 0>}
      localJointOffset=<8 0 0>
     }]
   }
   n52 = Anim Dynamics - Bone: dyn_shldr_f_l(ComponentPose=@n51, Alpha=0.3, ExternalForce="0,0,0") {
    angularDampingOverride=0.999
    angularSpringConstant=3000
    bAngularSpring=true
    bGravityOverrideInSimSpace=true
    bOverrideAngularDamping=true
    bUseGravityOverride=true
    gravityScale=0.35
    simulationSpace=RootRelative
    alphaBoolBlend {blendOption=Linear}
    boundBone {boneName=dyn_shldr_f_l}
    componentLinearAccScale=<1 1 1>
    componentLinearVelScale=<1 1 1>
    gravityOverride {z=-980}
    physicsBodyDefinitions [{
      sphereCollisionRadius=0
      boundBone {boneName=dyn_shldr_f_l}
      boxExtents=<25 15 15>
      constraintSetup {angularLimitsMax=<40 40 40>, angularLimitsMin=<-40 -40 -40>, angularTarget=<1 0 0>}
      localJointOffset=<8 0 0>
     }]
   }
   n53 = Anim Dynamics - Bone: dyn_shldr_d_l(ComponentPose=@n52, Alpha=0.16, ExternalForce="0,0,0") {
    angularDampingOverride=0.999
    angularSpringConstant=1500
    bAngularSpring=true
    bGravityOverrideInSimSpace=true
    bOverrideAngularDamping=true
    bUseGravityOverride=true
    gravityScale=0.35
    simulationSpace=RootRelative
    alphaBoolBlend {blendOption=Linear}
    boundBone {boneName=dyn_shldr_d_l}
    componentLinearAccScale=<1 1 1>
    componentLinearVelScale=<1 1 1>
    gravityOverride {z=-980}
    physicsBodyDefinitions [{
      sphereCollisionRadius=0
      boundBone {boneName=dyn_shldr_d_l}
      boxExtents=<25 15 15>
      constraintSetup {angularLimitsMax=<40 40 40>, angularLimitsMin=<-40 -40 -40>, angularTarget=<1 0 0>}
      localJointOffset=<8 0 0>
     }]
   }
   n54 = Anim Dynamics - Bone: dyn_shldr_c_l(ComponentPose=@n53, Alpha=0.15, ExternalForce="0,0,0") {
    angularDampingOverride=0.999
    angularSpringConstant=3000
    bAngularSpring=true
    bGravityOverrideInSimSpace=true
    bOverrideAngularDamping=true
    bUseGravityOverride=true
    gravityScale=0.35
    simulationSpace=RootRelative
    alphaBoolBlend {blendOption=Linear}
    boundBone {boneName=dyn_shldr_c_l}
    componentLinearAccScale=<1 1 1>
    componentLinearVelScale=<1 1 1>
    gravityOverride {z=-980}
    physicsBodyDefinitions [{
      sphereCollisionRadius=0
      boundBone {boneName=dyn_shldr_c_l}
      boxExtents=<25 15 15>
      constraintSetup {angularLimitsMax=<40 40 40>, angularLimitsMin=<-40 -40 -40>, angularTarget=<1 0 0>}
      localJointOffset=<8 0 0>
     }]
   }
   n55 = Anim Dynamics - Bone: dyn_shldr_g_l(ComponentPose=@n54, Alpha=0.23, ExternalForce="0,0,0") {
    angularDampingOverride=0.999
    angularSpringConstant=1500
    bAngularSpring=true
    bGravityOverrideInSimSpace=true
    bOverrideAngularDamping=true
    bUseGravityOverride=true
    gravityScale=0.35
    simulationSpace=RootRelative
    alphaBoolBlend {blendOption=Linear}
    boundBone {boneName=dyn_shldr_g_l}
    componentLinearAccScale=<1 1 1>
    componentLinearVelScale=<1 1 1>
    gravityOverride {z=-980}
    physicsBodyDefinitions [{
      sphereCollisionRadius=0
      boundBone {boneName=dyn_shldr_g_l}
      boxExtents=<25 15 15>
      constraintSetup {angularLimitsMax=<40 40 40>, angularLimitsMin=<-40 -40 -40>, angularTarget=<1 0 0>}
      localJointOffset=<8 0 0>
     }]
   }
   n56 = Anim Dynamics - Bone: dyn_shldr_h_l(ComponentPose=@n55, Alpha=0.3, ExternalForce="0,0,0") {
    angularDampingOverride=0.999
    angularSpringConstant=1500
    bAngularSpring=true
    bGravityOverrideInSimSpace=true
    bOverrideAngularDamping=true
    bUseGravityOverride=true
    gravityScale=0.35
    simulationSpace=RootRelative
    alphaBoolBlend {blendOption=Linear}
    boundBone {boneName=dyn_shldr_h_l}
    componentLinearAccScale=<1 1 1>
    componentLinearVelScale=<1 1 1>
    gravityOverride {z=-980}
    physicsBodyDefinitions [{
      sphereCollisionRadius=0
      boundBone {boneName=dyn_shldr_h_l}
      boxExtents=<25 15 15>
      constraintSetup {angularLimitsMax=<40 40 40>, angularLimitsMin=<-40 -40 -40>, angularTarget=<1 0 0>}
      localJointOffset=<8 0 0>
     }]
   }
   n57 = Anim Dynamics - Bone: dyn_shldr_a_r(ComponentPose=@n56, Alpha=0.2, ExternalForce="0,0,0") {
    angularDampingOverride=0.999
    angularSpringConstant=3000
    bAngularSpring=true
    bGravityOverrideInSimSpace=true
    bOverrideAngularDamping=true
    bUseGravityOverride=true
    gravityScale=0.35
    simulationSpace=RootRelative
    alphaBoolBlend {blendOption=Linear}
    boundBone {boneName=dyn_shldr_a_r}
    componentLinearAccScale=<1 1 1>
    componentLinearVelScale=<1 1 1>
    gravityOverride {z=-980}
    physicsBodyDefinitions [{
      sphereCollisionRadius=0
      boundBone {boneName=dyn_shldr_a_r}
      boxExtents=<25 15 15>
      constraintSetup {angularLimitsMax=<40 40 40>, angularLimitsMin=<-40 -40 -40>, angularTarget=<1 0 0>}
      localJointOffset=<8 0 0>
     }]
   }
   n58 = Anim Dynamics - Bone: dyn_shldr_f_r(ComponentPose=@n57, Alpha=0.1, ExternalForce="0,0,0") {
    angularDampingOverride=0.999
    angularSpringConstant=1500
    bAngularSpring=true
    bGravityOverrideInSimSpace=true
    bOverrideAngularDamping=true
    bUseGravityOverride=true
    gravityScale=0.35
    simulationSpace=RootRelative
    alphaBoolBlend {blendOption=Linear}
    boundBone {boneName=dyn_shldr_f_r}
    componentLinearAccScale=<1 1 1>
    componentLinearVelScale=<1 1 1>
    gravityOverride {z=-980}
    physicsBodyDefinitions [{
      sphereCollisionRadius=0
      boundBone {boneName=dyn_shldr_f_r}
      boxExtents=<25 15 15>
      constraintSetup {angularLimitsMax=<40 40 40>, angularLimitsMin=<-40 -40 -40>, angularTarget=<1 0 0>}
      localJointOffset=<8 0 0>
     }]
   }
   n59 = Anim Dynamics - Bone: dyn_shldr_c_r(ComponentPose=@n58, Alpha=0.23, ExternalForce="0,0,0") {
    angularDampingOverride=0.999
    angularSpringConstant=1500
    bAngularSpring=true
    bGravityOverrideInSimSpace=true
    bOverrideAngularDamping=true
    bUseGravityOverride=true
    gravityScale=0.35
    simulationSpace=RootRelative
    alphaBoolBlend {blendOption=Linear}
    boundBone {boneName=dyn_shldr_c_r}
    componentLinearAccScale=<1 1 1>
    componentLinearVelScale=<1 1 1>
    gravityOverride {z=-980}
    physicsBodyDefinitions [{
      sphereCollisionRadius=0
      boundBone {boneName=dyn_shldr_c_r}
      boxExtents=<25 15 15>
      constraintSetup {angularLimitsMax=<40 40 40>, angularLimitsMin=<-40 -40 -40>, angularTarget=<1 0 0>}
      localJointOffset=<8 0 0>
     }]
   }
   n60 = Anim Dynamics - Bone: dyn_shldr_g_r(ComponentPose=@n59, Alpha=0.3, ExternalForce="0,0,0") {
    angularDampingOverride=0.999
    angularSpringConstant=3000
    bAngularSpring=true
    bGravityOverrideInSimSpace=true
    bOverrideAngularDamping=true
    bUseGravityOverride=true
    gravityScale=0.35
    simulationSpace=RootRelative
    alphaBoolBlend {blendOption=Linear}
    boundBone {boneName=dyn_shldr_g_r}
    componentLinearAccScale=<1 1 1>
    componentLinearVelScale=<1 1 1>
    gravityOverride {z=-980}
    physicsBodyDefinitions [{
      sphereCollisionRadius=0
      boundBone {boneName=dyn_shldr_g_r}
      boxExtents=<25 15 15>
      constraintSetup {angularLimitsMax=<40 40 40>, angularLimitsMin=<-40 -40 -40>, angularTarget=<1 0 0>}
      localJointOffset=<8 0 0>
     }]
   }
   n61 = Anim Dynamics - Bone: dyn_shldr_d_r(ComponentPose=@n60, Alpha=0.22, ExternalForce="0,0,0") {
    angularDampingOverride=0.999
    angularSpringConstant=3000
    bAngularSpring=true
    bGravityOverrideInSimSpace=true
    bOverrideAngularDamping=true
    bUseGravityOverride=true
    gravityScale=0.35
    simulationSpace=RootRelative
    alphaBoolBlend {blendOption=Linear}
    boundBone {boneName=dyn_shldr_d_r}
    componentLinearAccScale=<1 1 1>
    componentLinearVelScale=<1 1 1>
    gravityOverride {z=-980}
    physicsBodyDefinitions [{
      sphereCollisionRadius=0
      boundBone {boneName=dyn_shldr_d_r}
      boxExtents=<25 15 15>
      constraintSetup {angularLimitsMax=<40 40 40>, angularLimitsMin=<-40 -40 -40>, angularTarget=<1 0 0>}
      localJointOffset=<8 0 0>
     }]
   }
   n62 = Anim Dynamics - Bone: dyn_shldr_h_r(ComponentPose=@n61, Alpha=0.27, ExternalForce="0,0,0") {
    angularDampingOverride=0.999
    angularSpringConstant=1500
    bAngularSpring=true
    bGravityOverrideInSimSpace=true
    bOverrideAngularDamping=true
    bUseGravityOverride=true
    gravityScale=0.35
    simulationSpace=RootRelative
    alphaBoolBlend {blendOption=Linear}
    boundBone {boneName=dyn_shldr_h_r}
    componentLinearAccScale=<1 1 1>
    componentLinearVelScale=<1 1 1>
    gravityOverride {z=-980}
    physicsBodyDefinitions [{
      sphereCollisionRadius=0
      boundBone {boneName=dyn_shldr_h_r}
      boxExtents=<25 15 15>
      constraintSetup {angularLimitsMax=<40 40 40>, angularLimitsMin=<-40 -40 -40>, angularTarget=<1 0 0>}
      localJointOffset=<8 0 0>
     }]
   }
   n63 = "Constraint - dyn_buttZipper_MSTR (PPPPPP)"!(
    ComponentPose=@n62
    Alpha=1
    ConstraintWeights_0=0.2464
    ConstraintWeights_1:float=0.0004
    ConstraintWeights_2:float=0.056
    ConstraintWeights_3:float=0.0006
    ConstraintWeights_4:float=0.6959
    ConstraintWeights_5:float=0.0006
   ) -> (Pose) {
    alphaBoolBlend {blendOption=Linear}
    boneToModify {boneName=dyn_buttZipper_MSTR}
    constraintSetup [
     {targetBone {boneName=pelvis}}
     {targetBone {boneName=spine_01}}
     {targetBone {boneName=dyn_coat_bk_l_01}}
     {targetBone {boneName=dyn_coat_bk_l_02}}
     {targetBone {boneName=dyn_coat_bk_01}}
     {targetBone {boneName=dyn_coat_bk_02}}
    ]
    constraintWeights [1,1,1,1,1,1]
   }
   n64 = "Constraint - dyn_shldr_c_base_r (PP)"!(ComponentPose=@n63, Alpha=1, ConstraintWeights_0=1, ConstraintWeights_1:float=0) -> (Pose) {
    alphaBoolBlend {blendOption=Linear}
    boneToModify {boneName=dyn_shldr_c_base_r}
    constraintSetup [{targetBone {boneName=upperarm_r}}, {targetBone {boneName=clavicle_r}}]
    constraintWeights [1,1]
   }
   n65 = "Transform (Modify) Bone - Bone: dyn_shldr_c_base_r"(ComponentPose=@n64, Alpha=Get upperarm_rot_r() -> (upperarm_rot_r:double), Translation="0,0,0", Rotation="-50,0,0", Scale="1,1,1") {
    rotationMode=BMM_Additive
    rotationSpace=BCS_BoneSpace
    scaleMode=BMM_Additive
    scaleSpace=BCS_BoneSpace
    translationMode=BMM_Additive
    translationSpace=BCS_BoneSpace
    alphaBoolBlend {blendOption=Linear}
    boneToModify {boneName=dyn_shldr_c_base_r}
    rotation {pitch=-50}
   }
   Output Pose(Result=Component To Local(ComponentPose=@n65)) {layerGroup=None}
   n66 = Control Rig(Source=Copy Pose From Mesh() {bUseAttachedParent=true}) {
    alphaBoolBlend {blendOption=Linear}
    controlRigAssetReference {blueprintRigClass=&/M_MED_Despair_DeformRig_Master_ControlRig_new.M_MED_Despair_DeformRig_Master_ControlRig_new_C}
   }
   n67 = Control Rig(Source=@n66) {
    alphaBoolBlend {blendOption=Linear}
    controlRigAssetReference {blueprintRigClass=&/M_MED_Despair_CtrlRig.M_MED_Despair_CtrlRig_C}
   }
   n68 = "Constraint - dyn_l_thighZipper_MSTR (PPPP)"!(
    ComponentPose=Local To Component(LocalPose=@n67)
    Alpha=1
    ConstraintWeights_0=0.0207
    ConstraintWeights_1:float=0.2449
    ConstraintWeights_2:float=0.68
    ConstraintWeights_3:float=0.0543
   ) -> (Pose) {
    alphaBoolBlend {blendOption=Linear}
    boneToModify {boneName=dyn_l_thighZipper_MSTR}
    constraintSetup [
     {targetBone {boneName=pelvis}}
     {targetBone {boneName=thigh_l}}
     {targetBone {boneName=thigh_twist_01_l}}
     {targetBone {boneName=deform_groin_l}}
    ]
    constraintWeights [1,1,1,1]
   }
   n69 = "Constraint - dyn_r_thighZipper_MSTR (PPPP)"!(
    ComponentPose=@n68
    Alpha=1
    ConstraintWeights_0=0.0215
    ConstraintWeights_1:float=0.2389
    ConstraintWeights_2:float=0.6825
    ConstraintWeights_3:float=0.057
   ) -> (Pose) {
    alphaBoolBlend {blendOption=Linear}
    boneToModify {boneName=dyn_r_thighZipper_MSTR}
    constraintSetup [
     {targetBone {boneName=pelvis}}
     {targetBone {boneName=thigh_r}}
     {targetBone {boneName=thigh_twist_01_r}}
     {targetBone {boneName=deform_groin_r}}
    ]
    constraintWeights [1,1,1,1]
   }
   n70 = "Constraint - dyn_chestZipper_MSTR (PPPPPPPP)"!(
    ComponentPose=@n69
    Alpha=1
    ConstraintWeights_0=0.0562
    ConstraintWeights_1:float=0.1469
    ConstraintWeights_2:float=0.3874
    ConstraintWeights_3:float=0.4845
    ConstraintWeights_4:float=0.0579
    ConstraintWeights_5:float=0.057
    ConstraintWeights_6:float=0.0063
    ConstraintWeights_7:float=0.0579
   ) -> (Pose) {
    alphaBoolBlend {blendOption=Linear}
    boneToModify {boneName=dyn_chestZipper_MSTR}
    constraintSetup [
     {targetBone {boneName=spine_02}}
     {targetBone {boneName=spine_03}}
     {targetBone {boneName=spine_04}}
     {targetBone {boneName=spine_05}}
     {targetBone {boneName=clavicle_r}}
     {targetBone {boneName=deform_pec_r}}
     {targetBone {boneName=deform_pec_l}}
     {targetBone {boneName=clavicle_r}}
    ]
    constraintWeights [1,1,1,1,1,1,1,1]
   }
   n71 = "Constraint - dyn_gernade_01_MSTR (PPPP)"!(
    ComponentPose=@n70
    Alpha=1
    ConstraintWeights_0=0.551
    ConstraintWeights_1:float=0.1565
    ConstraintWeights_2:float=0.1848
    ConstraintWeights_3:float=0.1076
   ) -> (Pose) {
    alphaBoolBlend {blendOption=Linear}
    boneToModify {boneName=dyn_gernade_01_MSTR}
    constraintSetup [{targetBone {boneName=pelvis}}, {targetBone {boneName=spine_01}}, {targetBone {boneName=spine_02}}, {targetBone {boneName=spine_03}}]
    constraintWeights [1,1,1,1]
   }
   n72 = "Constraint - dyn_gernade_02_MSTR (PPPP)"!(
    ComponentPose=@n71
    Alpha=1
    ConstraintWeights_0=0.5608
    ConstraintWeights_1:float=0.1531
    ConstraintWeights_2:float=0.1824
    ConstraintWeights_3:float=0.1037
   ) -> (Pose) {
    alphaBoolBlend {blendOption=Linear}
    boneToModify {boneName=dyn_gernade_02_MSTR}
    constraintSetup [{targetBone {boneName=pelvis}}, {targetBone {boneName=spine_01}}, {targetBone {boneName=spine_02}}, {targetBone {boneName=spine_03}}]
    constraintWeights [1,1,1,1]
   }
   n73 = "Constraint - dyn_gernade_03_MSTR (PPPP)"!(
    ComponentPose=@n72
    Alpha=1
    ConstraintWeights_0=0.5535
    ConstraintWeights_1:float=0.1548
    ConstraintWeights_2:float=0.1868
    ConstraintWeights_3:float=0.1048
   ) -> (Pose) {
    alphaBoolBlend {blendOption=Linear}
    boneToModify {boneName=dyn_gernade_03_MSTR}
    constraintSetup [{targetBone {boneName=pelvis}}, {targetBone {boneName=spine_01}}, {targetBone {boneName=spine_02}}, {targetBone {boneName=spine_03}}]
    constraintWeights [1,1,1,1]
   }
   n74 = "Constraint - dyn_gernade_04_MSTR (PPPPP)"!(
    ComponentPose=@n73
    Alpha=1
    ConstraintWeights_0=0.5331
    ConstraintWeights_1:float=0.1605
    ConstraintWeights_2:float=0.1953
    ConstraintWeights_3:float=0.1104
    ConstraintWeights_4:float=0.0006
   ) -> (Pose) {
    alphaBoolBlend {blendOption=Linear}
    boneToModify {boneName=dyn_gernade_04_MSTR}
    constraintSetup [
     {targetBone {boneName=pelvis}}
     {targetBone {boneName=spine_01}}
     {targetBone {boneName=spine_02}}
     {targetBone {boneName=spine_03}}
     {targetBone {boneName=spine_04}}
    ]
    constraintWeights [1,1,1,1,1]
   }
   n75 = "Constraint - dyn_torsoTassel_MSTR (PPPPPPPP)"!(
    ComponentPose=@n74
    Alpha=1
    ConstraintWeights_0=0.0495
    ConstraintWeights_1:float=0.1141
    ConstraintWeights_2:float=0.2452
    ConstraintWeights_3:float=0.3144
    ConstraintWeights_4:float=0.1244
    ConstraintWeights_5:float=0.092
    ConstraintWeights_6:float=0.0248
    ConstraintWeights_7:float=0.0356
   ) -> (Pose) {
    alphaBoolBlend {blendOption=Linear}
    boneToModify {boneName=dyn_torsoTassel_MSTR}
    constraintSetup [
     {targetBone {boneName=pelvis}}
     {targetBone {boneName=spine_01}}
     {targetBone {boneName=spine_02}}
     {targetBone {boneName=spine_03}}
     {targetBone {boneName=spine_04}}
     {targetBone {boneName=clavicle_l}}
     {targetBone {boneName=upperarm_twist_01_l}}
     {targetBone {boneName=deform_pec_l}}
    ]
    constraintWeights [1,1,1,1,1,1,1,1]
   }
   n76 = "Constraint - dyn_backTassel_MSTR (PPPPPPP)"!(
    ComponentPose=@n75
    Alpha=1
    ConstraintWeights_0=0.1855
    ConstraintWeights_1:float=0.0584
    ConstraintWeights_2:float=0.2232
    ConstraintWeights_3:float=0.4037
    ConstraintWeights_4:float=0.0606
    ConstraintWeights_5:float=0.0429
    ConstraintWeights_6:float=0.0257
   ) -> (Pose) {
    alphaBoolBlend {blendOption=Linear}
    boneToModify {boneName=dyn_backTassel_MSTR}
    constraintSetup [
     {targetBone {boneName=spine_03}}
     {targetBone {boneName=spine_04}}
     {targetBone {boneName=spine_05}}
     {targetBone {boneName=clavicle_r}}
     {targetBone {boneName=upperarm_twist_01_r}}
     {targetBone {boneName=deform_upperarm_up_r}}
     {targetBone {boneName=deform_upperarm_bk_r}}
    ]
    constraintWeights [1,1,1,1,1,1,1]
   }
   n77 = "Constraint - dyn_l_wristStrap_MSTR (PPPPPPPP)"!(
    ComponentPose=@n76
    Alpha=1
    ConstraintWeights_0=0.0495
    ConstraintWeights_1:float=0.1141
    ConstraintWeights_2:float=0.2452
    ConstraintWeights_3:float=0.3144
    ConstraintWeights_4:float=0.1244
    ConstraintWeights_5:float=0.092
    ConstraintWeights_6:float=0.0248
    ConstraintWeights_7:float=0.0356
   ) -> (Pose) {
    alphaBoolBlend {blendOption=Linear}
    boneToModify {boneName=dyn_l_wristStrap_MSTR}
    constraintSetup [
     {targetBone {boneName=lowerarm_l}}
     {targetBone {boneName=hand_l}}
     {targetBone {boneName=pinky_metacarpal_l}}
     {targetBone {boneName=thumb_01_l}}
     {targetBone {boneName=deform_wrist_up_l}}
     {targetBone {boneName=deform_wrist_dn_l}}
     {targetBone {boneName=lowerarm_twist_01_l}}
     {targetBone {boneName=lowerarm_twist_02_l}}
    ]
    constraintWeights [1,1,1,1,1,1,1,1]
   }
   Save cached pose 'PreSim'(Pose=Component To Local(ComponentPose=@n77)) {cacheName=PreSim}
  ]}
 EventGraph {nodes [
   Event BlueprintUpdateAnimation()
   > n1 = Sequence!(execute) -> (then_0, then_1, then_2:Exec, then_3:Exec)
   n2 = Gravity Override State Function DEPRECATED(
    execute=@n1.then_2
    CharacterPartAnimInstance=Self-Reference()
    GameTimeInSeconds=GetGameTimeInSeconds()
    GravityOverrideParams=Get GravityOverrideParams() -> (GravityOverrideParams:Gravity Override Params Struct Structure)
   )
   > Set GravityOverride(GravityOverride=@n2)
   n3 = Get Skel_Mesh()
   > Set thigh_rot_otr_l(thigh_rot_otr_l:double=GetAlphaFromReferenceTransformDelta(SkeletalMeshComponent=@n3, BoneName=thigh_l, BaseName=pelvis, DeltaComponentType=EulerY, InputRange=(X=20,Y=50), TwistAxis="1,0,0")) -> (Output_Get:double)
   > Set thigh_rot_otr_r(thigh_rot_otr_r:double=GetAlphaFromReferenceTransformDelta(SkeletalMeshComponent=@n3, BoneName=thigh_r, BaseName=pelvis, DeltaComponentType=EulerY, InputRange=(X=20,Y=50), TwistAxis="1,0,0")) -> (Output_Get:double)
   n4 = Get Skel_Mesh()
   > Set thigh_rot_inr_l(thigh_rot_inr_l:double=GetAlphaFromReferenceTransformDelta(SkeletalMeshComponent=@n4, BoneName=thigh_l, BaseName=pelvis, DeltaComponentType=EulerY, InputRange=(X=10,Y=-6), TwistAxis="1,0,0")) -> (Output_Get:double)
   > Set thigh_rot_inr_r(thigh_rot_inr_r:double=GetAlphaFromReferenceTransformDelta(SkeletalMeshComponent=@n4, BoneName=thigh_r, BaseName=pelvis, DeltaComponentType=EulerY, InputRange=(X=0,Y=-10), TwistAxis="1,0,0")) -> (Output_Get:double)
   n5 = Get Skel_Mesh()
   > Set upperarm_rot_l(upperarm_rot_l:double=GetAlphaFromReferenceTransformDelta(SkeletalMeshComponent=@n5, BoneName=upperarm_l, BaseName=clavicle_l, InputRange=(X=0,Y=80), TwistAxis="1,0,0")) -> (Output_Get:double)
   > Set upperarm_rot_r(upperarm_rot_r:double=GetAlphaFromReferenceTransformDelta(SkeletalMeshComponent=@n5, BoneName=upperarm_r, BaseName=clavicle_r, InputRange=(X=90,Y=120), TwistAxis="1,0,0")) -> (Output_Get:double)
   Reroute Node(InputPin=@n1.then_3)
   n6 = Get bIsPlayingEmote()
   n7 = Get bIsMoving2D()
   n8 = Get bIsCrouching()
   n9 = Branch(execute=@n1.then_1, Condition=AND Boolean!(A=NOT Boolean(A=@n7), B=NOT Boolean(A=@n8), C:bool=NOT Boolean(A=@n6)) -> (ReturnValue))
   > Set back skirt standing(back skirt standing:bool=true) -> (Output_Get:bool)
   Set back skirt standing(execute=@n9.else, back skirt standing=false)
   n10 = Get bIsMoving2D()
   n11 = Branch(execute=@n1.then_0, Condition=AND Boolean!(A=Get bIsCrouching(), B=NOT Boolean(A=@n10)) -> (ReturnValue))
   > Set still_crouch(still_crouch:bool=true) -> (Output_Get:bool)
   Set still_crouch(execute=@n11.else, still_crouch=false)
   n12 = Event OnBeginFalling()
   Set jumping(execute=[@n12.then, @n13.then], jumping:double=1) -> (Output_Get:double)
   n13 = Event OnBeginJumping()
   Event BlueprintInitializeAnimation()
   > Set Skel_Mesh(Skel_Mesh=GetOwnerSkeletalMeshComponent())
   n14 = Event OnEndFalling()
   Set jumping(execute=[@n14.then, @n15.then], jumping=0)
   n15 = Event OnEndJumping()
   n16 = GetOwnerSkeletalMeshComponent()
   n17 = BreakRotator(InRot=BreakTransform(InTransform=GetDeltaTransformFromRefPose(self=@n16, BoneName=upperarm_r, BaseName=clavicle_r)).Rotation)
   PrintString!(execute, InString="To String (Float)"(InDouble=@n17.Yaw), bPrintToScreen, bPrintToLog, TextColor, Duration, Key) -> (then)
   MapRangeClamped(Value=@n17.Yaw, InRangeA=90, InRangeB=120)
  ]}
]
variables [
 back skirt standing {type=boolean}
 GravityOverride {type=Vector}
 GravityOverrideParams {type=GravityOverrideParamsStruct}
 jumping {type=number}
 LoinClothBackRotator_01 {type=Rotator}
 LoinClothBackRotator_02 {type=Rotator}
 LoinClothRotator_01 {type=Rotator}
 LoinClothRotator_02 {type=Rotator}
 Skel_Mesh {type=SkeletalMeshComponent}
 still_crouch {type=boolean}
 thigh_rot_inr_l {type=number}
 thigh_rot_inr_r {type=number}
 thigh_rot_otr_l {type=number}
 thigh_rot_otr_r {type=number}
 upperarm_rot_l {type=number}
 upperarm_rot_r {type=number}
]
