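# Physics asset for the face accessory mesh of the male medium Despair player body, with capsule bodies for its dynamic hair-named bone chains and upper body.
# Six chains (three bottom, one side, two top) branch from dyn_hair_Main under the head; joint5 ends the bottom-b chain.
# Chain joints carry asymmetric swing and twist limits, slightly different on top-b.
# Head, neck, spine and clavicle capsules use angularly free constraints, as does the link from dyn_hair_Main to head.
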PhysicsAsset /BRCosmetics/Characters/Player/Male/Medium/Bodies/M_MED_Despair/Meshes/Parts/M_MED_Despair_FaceAcc_Physics
bodies {
 clavicle_l [{shapeType=Capsule, length=8.4972, radius=11.5028, shapeName=clavicle_l_capsule, center=<7.7479 0 0>, rotation=<rot 90 -87.106 -95.5291>}]
 clavicle_r [{shapeType=Capsule, length=9.3907, radius=10.6093, shapeName=clavicle_r_capsule, center=<-7.2394 0 0>, rotation=<rot 90 -87.106 -98.2372>}]
 dyn_btm_a_01 [{shapeType=Capsule, length=3.2657, radius=2.4603, shapeName=dyn_btm_a_01_capsule, center=<2.7306 0 3>, rotation=<rot 90 0 0>}]
 dyn_btm_a_02 [{shapeType=Capsule, length=3.2657, radius=2.4603, shapeName=dyn_btm_a_02_capsule, center=<2.7306 0 3>, rotation=<rot 90 0 0>}]
 dyn_btm_a_03 [{shapeType=Capsule, length=2.015, radius=3.711, shapeName=dyn_btm_a_03_capsule, center=<2.7306 0 3>, rotation=<rot 90 0 0>}]
 dyn_btm_a_04 [{shapeType=Capsule, length=2.015, radius=3.711, shapeName=dyn_btm_a_04_capsule, center=<2.7306 0 3>, rotation=<rot 90 0 0>}]
 dyn_btm_b_01 [{shapeType=Capsule, length=1.5447, radius=4.1813, shapeName=dyn_btm_b_01_capsule, center=<2.7306 0 3>, rotation=<rot 90 0 0>}]
 dyn_btm_b_02 [{shapeType=Capsule, length=1.5447, radius=4.1813, shapeName=dyn_btm_b_02_capsule, center=<2.7306 0 3>, rotation=<rot 90 0 0>}]
 dyn_btm_b_03 [{shapeType=Capsule, length=1.5447, radius=4.1813, shapeName=dyn_btm_b_03_capsule, center=<2.7306 0 3>, rotation=<rot 90 0 0>}]
 dyn_btm_c_01 [{shapeType=Capsule, length=1.0274, radius=4.6986, shapeName=dyn_btm_c_01_capsule, center=<2.7306 0 3>, rotation=<rot 90 0 0>}]
 dyn_btm_c_02 [{shapeType=Capsule, length=0, radius=6.4042, shapeName=dyn_btm_c_02_capsule, center=<2.7306 0 3>, rotation=<rot 90 0 0>}]
 dyn_btm_c_03 [{shapeType=Capsule, length=0, radius=6.4042, shapeName=dyn_btm_c_03_capsule, center=<2.7306 0 3>, rotation=<rot 90 0 0>}]
 dyn_btm_c_04 [{shapeType=Capsule, length=0, radius=6.4042, shapeName=dyn_btm_c_04_capsule, center=<2.7306 0 3>, rotation=<rot 90 0 0>}]
 dyn_hair_Main [{shapeType=Capsule, length=1, radius=1, shapeName=dyn_hair_Main_capsule, rotation=<rot 0 0 90>}]
 dyn_side_01 [{shapeType=Capsule, length=1.0274, radius=4.6986, shapeName=dyn_side_01_capsule, center=<2.7306 0 3>, rotation=<rot 90 0 0>}]
 dyn_side_02 [{shapeType=Capsule, length=1.0274, radius=4.6986, shapeName=dyn_side_02_capsule, center=<2.7306 0 3>, rotation=<rot 90 0 0>}]
 dyn_side_03 [{shapeType=Capsule, length=1.0274, radius=4.6986, shapeName=dyn_side_03_capsule, center=<2.7306 0 3>, rotation=<rot 90 0 0>}]
 dyn_top_a_01 [{shapeType=Capsule, length=2.5701, radius=6.1611, shapeName=dyn_top_a_01_capsule, center=<2.7306 0 3>, rotation=<rot 90 0 0>}]
 dyn_top_a_02 [{shapeType=Capsule, length=1.1406, radius=7.5906, shapeName=dyn_top_a_02_capsule, center=<2.7306 0 3>, rotation=<rot 90 0 0>}]
 dyn_top_a_03 [{shapeType=Capsule, length=0.9821, radius=7.7491, shapeName=dyn_top_a_03_capsule, center=<2.7306 0 3>, rotation=<rot 90 0 0>}]
 dyn_top_b_01 [{shapeType=Capsule, length=0, radius=8.5466, shapeName=dyn_top_b_01_capsule, center=<4.096 0 3>, rotation=<rot 90 0 0>}]
 dyn_top_b_02 [{shapeType=Capsule, length=0, radius=9.2401, shapeName=dyn_top_b_02_capsule, center=<4.096 0 3>, rotation=<rot 90 0 0>}]
 dyn_top_b_03 [{shapeType=Capsule, length=0, radius=9.2401, shapeName=dyn_top_b_03_capsule, center=<4.096 0 3>, rotation=<rot 90 0 0>}]
 head [{shapeType=Capsule, length=5.1829, radius=9.5903, shapeName=head_capsule, rotation=<rot 90 0 0>}]
 joint5 [{shapeType=Capsule, length=1.0274, radius=4.6986, shapeName=joint5_capsule, center=<2.7306 0 3>, rotation=<rot 90 0 0>}]
 neck_01 [{shapeType=Capsule, length=1.1107, radius=8.4534, shapeName=neck_01_capsule}]
 neck_02 [{shapeType=Capsule, length=0.6222, radius=8.9418, shapeName=neck_02_capsule, rotation=<rot 0 0 -0.0721>}]
 spine_05 [
  {shapeType=Capsule, length=11.8517, radius=12.2196, shapeName=spine_05_capsule}
  {shapeType=Capsule, length=13.7149, radius=10.3564, shapeName=spine_05_capsule, center=<9.0342 -0.6493 0>}
 ]
}
constraints [
 {
  bone1Name=head
  bone2Name=neck_02
  swing1LimitDegrees=45
  swing1Motion=Free
  swing2LimitDegrees=45
  swing2Motion=Free
  twistLimitDegrees=45
  twistMotion=Free
 }
 {
  bone1Name=neck_02
  bone2Name=neck_01
  swing1LimitDegrees=45
  swing1Motion=Free
  swing2LimitDegrees=45
  swing2Motion=Free
  twistLimitDegrees=45
  twistMotion=Free
 }
 {
  bone1Name=neck_01
  bone2Name=spine_05
  swing1LimitDegrees=45
  swing1Motion=Free
  swing2LimitDegrees=45
  swing2Motion=Free
  twistLimitDegrees=45
  twistMotion=Free
 }
 {
  bone1Name=dyn_hair_Main
  bone2Name=head
  swing1LimitDegrees=45
  swing1Motion=Free
  swing2LimitDegrees=45
  swing2Motion=Free
  twistLimitDegrees=45
  twistMotion=Free
 }
 {bone1Name=dyn_btm_a_01, bone2Name=dyn_hair_Main, swing1LimitDegrees=35, swing2LimitDegrees=80, twistLimitDegrees=15}
 {bone1Name=dyn_btm_a_02, bone2Name=dyn_btm_a_01, swing1LimitDegrees=35, swing2LimitDegrees=80, twistLimitDegrees=15}
 {bone1Name=dyn_btm_a_03, bone2Name=dyn_btm_a_02, swing1LimitDegrees=35, swing2LimitDegrees=80, twistLimitDegrees=15}
 {bone1Name=dyn_btm_a_04, bone2Name=dyn_btm_a_03, swing1LimitDegrees=35, swing2LimitDegrees=80, twistLimitDegrees=15}
 {bone1Name=dyn_btm_b_01, bone2Name=dyn_hair_Main, swing1LimitDegrees=35, swing2LimitDegrees=80, twistLimitDegrees=15}
 {bone1Name=dyn_btm_b_02, bone2Name=dyn_btm_b_01, swing1LimitDegrees=35, swing2LimitDegrees=80, twistLimitDegrees=15}
 {bone1Name=dyn_btm_b_03, bone2Name=dyn_btm_b_02, swing1LimitDegrees=35, swing2LimitDegrees=80, twistLimitDegrees=15}
 {bone1Name=joint5, bone2Name=dyn_btm_b_03, swing1LimitDegrees=35, swing2LimitDegrees=80, twistLimitDegrees=15}
 {bone1Name=dyn_btm_c_01, bone2Name=dyn_hair_Main, swing1LimitDegrees=35, swing2LimitDegrees=80, twistLimitDegrees=15}
 {bone1Name=dyn_btm_c_02, bone2Name=dyn_btm_c_01, swing1LimitDegrees=35, swing2LimitDegrees=80, twistLimitDegrees=15}
 {bone1Name=dyn_btm_c_03, bone2Name=dyn_btm_c_02, swing1LimitDegrees=35, swing2LimitDegrees=80, twistLimitDegrees=15}
 {bone1Name=dyn_btm_c_04, bone2Name=dyn_btm_c_03, swing1LimitDegrees=35, swing2LimitDegrees=80, twistLimitDegrees=15}
 {bone1Name=dyn_side_01, bone2Name=dyn_hair_Main, swing1LimitDegrees=35, swing2LimitDegrees=80, twistLimitDegrees=15}
 {bone1Name=dyn_side_02, bone2Name=dyn_side_01, swing1LimitDegrees=35, swing2LimitDegrees=80, twistLimitDegrees=15}
 {bone1Name=dyn_side_03, bone2Name=dyn_side_02, swing1LimitDegrees=35, swing2LimitDegrees=80, twistLimitDegrees=15}
 {bone1Name=dyn_top_a_01, bone2Name=dyn_hair_Main, swing1LimitDegrees=35, swing2LimitDegrees=80, twistLimitDegrees=15}
 {bone1Name=dyn_top_a_02, bone2Name=dyn_top_a_01, swing1LimitDegrees=35, swing2LimitDegrees=80, twistLimitDegrees=15}
 {bone1Name=dyn_top_a_03, bone2Name=dyn_top_a_02, swing1LimitDegrees=35, swing2LimitDegrees=80, twistLimitDegrees=15}
 {bone1Name=dyn_top_b_01, bone2Name=dyn_hair_Main, swing1LimitDegrees=30, swing2LimitDegrees=80, twistLimitDegrees=30}
 {bone1Name=dyn_top_b_02, bone2Name=dyn_top_b_01, swing1LimitDegrees=30, swing2LimitDegrees=80, twistLimitDegrees=30}
 {bone1Name=dyn_top_b_03, bone2Name=dyn_top_b_02, swing1LimitDegrees=30, swing2LimitDegrees=80, twistLimitDegrees=30}
 {
  bone1Name=clavicle_r
  bone2Name=spine_05
  swing1LimitDegrees=45
  swing1Motion=Free
  swing2LimitDegrees=45
  swing2Motion=Free
  twistLimitDegrees=45
  twistMotion=Free
 }
 {
  bone1Name=clavicle_l
  bone2Name=spine_05
  swing1LimitDegrees=45
  swing1Motion=Free
  swing2LimitDegrees=45
  swing2Motion=Free
  twistLimitDegrees=45
  twistMotion=Free
 }
]
